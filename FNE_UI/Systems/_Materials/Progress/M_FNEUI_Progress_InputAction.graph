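# Translucent UI material for an input-action progress indicator, drawing a radial sweep on a circle or rounded-box shape.
# UVs rotated by Start Point become a direction-switchable angle; Floor(value + angle) yields hard masks for Range and Percentage scaled by Range.
# The progress mask blends background and progress color and alpha, and adds a separately scaled fill shape onto the base shape.
# Opacity is clipped to the Range arc and multiplied by Hide Progress.
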Material /FNE_UI/Systems/_Materials/Progress/M_FNEUI_Progress_InputAction
graph {
 properties {blendMode=BLEND_Translucent, materialDomain=MD_UI, pixelDepthOffsetMode=PDOM_Legacy, usedShadingModels=Default Lit}
 nodes [
  n1 = VectorParameter() {group=Colors, parameterName=Color Background, sortPriority=2, defaultValue=<rgba 0.08 0 0.08 0.5>}
  n2 = VectorParameter() {group=Colors, parameterName=Color Progress, sortPriority=1, defaultValue=<rgba 1 0 1 1>}
  n3 = NamedRerouteUsage()
  LinearInterpolate(A=@n1.RGB, B=@n2.RGB, Alpha=@n3)
  n4 = Constant() {r=1}
  n5 = ScalarParameter() {defaultValue=0.01, group=Shape, parameterName=Circle Hardness, sortPriority=22}
  n6 = MaterialFunctionCall!(
   UV=MaterialFunctionCall(Texture Scale=ScalarParameter() {defaultValue=0.95, group=FillShape, parameterName=Fill Circle Scale, sortPriority=21}) {materialFunction=&ScaleUVsByCenter}.UVs
   Size=@n4
   GlowMin
   GlowMax=@n5
   OutlineThickness
  ) {materialFunction=/FNE_UI/Systems/_MaterialFunctions/Shapes/MF_FNEUI_Circle}
  n7 = ScalarParameter() {defaultValue=0.85, group=FillShape, parameterName=Fill Box Scale X, sortPriority=31}
  n8 = ScalarParameter() {defaultValue=0.85, group=FillShape, parameterName=Fill Box Scale Y, sortPriority=31}
  n9 = ScalarParameter() {defaultValue=0.95, group=Shape, parameterName=Box Hardness, sortPriority=31}
  n10 = ScalarParameter() {defaultValue=0.5, group=Shape, parameterName=Box CornerRadius}
  n11 = MaterialFunctionCall!(
   UVs=MaterialFunctionCall(Texture Scale=AppendVector(A=@n7, B=@n8)) {materialFunction=&ScaleUVsByCenter}.UVs
   Hardness=@n9
   Corner Roundness=@n10
   UV Scale from center
   Aspect Ratio
   Keep Aspect Ratio
  ) {materialFunction=/FNE_UI/Systems/_MaterialFunctions/Shapes/MF_FNEUI_Box_Cheaper}
  n12 = ScalarParameter() {defaultValue=1, desc="0 = Circle\r\n1 = Rounded Box", group=Shape, parameterName=Shape Switch, sliderMax=1, sortPriority=20}
  n13 = Saturate(None=@n12)
  n14 = MaterialFunctionCall(Texture Scale=ScalarParameter() {defaultValue=0.85, group=Shape, parameterName=Circle Scale, sortPriority=21}) {desc=Scale To Fit, materialFunction=&ScaleUVsByCenter}
  n15 = MaterialFunctionCall(Texture Scale=AppendVector(
    A=ScalarParameter() {defaultValue=0.75, group=Shape, parameterName=Box Scale X, sortPriority=31}
    B=ScalarParameter() {defaultValue=0.75, group=Shape, parameterName=Box Scale Y, sortPriority=31}
   )) {desc=Scale To Fit, materialFunction=&ScaleUVsByCenter}
  n16 = Add(
   A=Multiply(A=LinearInterpolate(A=@n6.Fill, B=@n11.Result, Alpha=@n13), B=NamedRerouteUsage())
   B=LinearInterpolate(
    A=MaterialFunctionCall(UV=@n14.UVs, Size=@n4, GlowMax=@n5) {materialFunction=&MF_FNEUI_Circle}.Fill
    B=MaterialFunctionCall(UVs=@n15.UVs, Hardness=@n9, Corner Roundness=@n10) {materialFunction=&MF_FNEUI_Box_Cheaper}.Result
    Alpha=@n13
   )
  )
  n17 = ScalarParameter() {defaultValue=0.5, group=Progress, parameterName=Range, sliderMax=1, sortPriority=12}
  n18 = MaterialFunctionCall!(
   Vector or UVs=MaterialFunctionCall(
    UVs=TextureCoordinate()
    "Rotation Angle (0-1)"=ScalarParameter() {defaultValue=0.5, group=Progress, parameterName=Start Point, sliderMax=1, sortPriority=11}
   ) {materialFunction=&CustomRotator}.Rotated Values
   Swizzle Coordinate Output
  ) {materialFunction=/FNE_UI/Systems/_MaterialFunctions/Basics/MF_FNEUI_VectorToRadial_Cheaper}
  n19 = ScalarParameter() {defaultValue=1, desc="0 = Counter Clockwise\r\n1 = Clockwise (default)", group=Progress, parameterName=Clockwise, sliderMax=1}
  n20 = LinearInterpolate(A=OneMinus(None=@n18.Vector Converted to Angle), B=@n18.Vector Converted to Angle, Alpha=Saturate(None=Ceil(None=@n19)))
  n21 = Floor(None=Add(A=@n17, B=@n20))
  Multiply(
   A=Multiply(A=LinearInterpolate(A=@n1.A, B=@n2.A, Alpha=@n3), B=Multiply(A=@n16, B=@n21))
   B=ScalarParameter() {defaultValue=1, group=Progress, parameterName=Hide Progress, sliderMax=1}
  )
  n22 = Multiply(A=@n17, B=ScalarParameter() {defaultValue=0.35, group=Progress, parameterName=Percentage, sliderMax=1, sortPriority=10})
  NamedRerouteDeclaration(None=Floor(None=Add(A=@n22, B=@n20))) {name=ProgressFill, nodeColor=<rgba 0 0 0 1>}
 ]
}
