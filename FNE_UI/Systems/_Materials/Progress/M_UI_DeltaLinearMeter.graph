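# Translucent UI material for a linear progress meter that adds a delta layer showing progress lost or gained.
# Progress and Delta parameters feed scale and translate UV functions that shape box-function fills, with math for remapping a custom min/max range.
# Delta sign switches the scale pivot, and AnimateDelta is exposed to animate between progress and delta values.
# The box masks blend background, progress and delta colors into a final color.
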Material /FNE_UI/Systems/_Materials/Progress/M_UI_DeltaLinearMeter
graph {
 properties {bAutomaticallySetUsageInEditor=false, blendMode=BLEND_Translucent, materialDomain=MD_UI, usedShadingModels=Default Lit}
 nodes [
  n1 = NamedRerouteUsage()
  n2 = NamedRerouteUsage()
  n3 = If(A=NamedRerouteUsage(), B=@n1, "A > B"=@n1, "A == B"=@n1, "A < B"=Multiply(A=Subtract(B=@n2)) {constB=-1})
  n4 = NamedRerouteUsage()
  n5 = NamedRerouteUsage()
  n6 = If(A=Subtract(A=@n4, B=NamedRerouteUsage()), B=@n5, "A > B"=@n5, "A == B"=@n5, "A < B"=Multiply(A=Subtract(B=@n4)) {constB=-1})
  n7 = NamedRerouteUsage()
  n8 = MaterialFunctionCall!(
   Input UVs
   Translate=MaterialFunctionCall(X=LinearInterpolate(A=@n3, B=@n6, Alpha=@n7), Y=Constant()) {materialFunction=&MakeFloat2}.Result
  ) {materialFunction=/FNE_UI/Systems/_MaterialFunctions/Transforms/MF_UI_Translate}
  n9 = NamedRerouteUsage()
  n10 = Abs(None=NamedRerouteUsage())
  n11 = NamedRerouteUsage()
  n12 = MaterialFunctionCall(X=Multiply(A=LinearInterpolate(A=@n9, B=@n10, Alpha=@n11) {constA=1}, B=NamedRerouteUsage()), Y=Constant() {r=1}) {materialFunction=&MakeFloat2}
  n13 = Constant2Vector() {r=1}
  n14 = If(A=NamedRerouteUsage(), "A > B"=@n13, "A == B"=@n13, "A < B"=Constant2Vector())
  n15 = Constant()
  NamedRerouteDeclaration(None=MaterialFunctionCall!(
    UVs=MaterialFunctionCall!(UVs=@n8.UVs raw, Scale=@n12.Result, Scale center=@n14) {materialFunction=/FNE_UI/Systems/_MaterialFunctions/Transforms/MF_UI_Scale}.Result
    Size=Constant() {r=1}
    Outline Thickness
    Glow Min
    Glow Max=@n15
    Corner Roundness=@n15
    Keep Aspect Ratio
   ) {materialFunction=/FNE_UI/Systems/_MaterialFunctions/Shapes/MF_FNEUI_Box}.Fill) {name=DeltaUV, nodeColor=<rgba 1 0.7734 0 1>}
  n16 = MaterialFunctionCall(X=LinearInterpolate(A=NamedRerouteUsage(), B=NamedRerouteUsage(), Alpha=NamedRerouteUsage()), Y=Constant() {r=1}) {materialFunction=&MakeFloat2}
  n17 = Constant2Vector()
  n18 = Constant()
  NamedRerouteDeclaration(None=MaterialFunctionCall(
    UVs=MaterialFunctionCall(Scale=@n16.Result, Scale center=@n17) {materialFunction=&MF_UI_Scale}.Result
    Size=Constant() {r=1}
    Glow Max=@n18
    Corner Roundness=@n18
   ) {materialFunction=&MF_FNEUI_Box}.Fill) {name=ProgressUV, nodeColor=<rgba 0.4531 0 1 1>}
  n19 = ScalarParameter() {
   defaultValue=0.25
   desc="A layer to showcase the difference between the previous progress and current progress.\r\n\r\nDelta more than 0 shows progress lost\r\nDelta less than 0 shows progress gained"
   group="[1] Basics"
   parameterName=Delta
  }
  NamedRerouteDeclaration(None=@n19) {name=Delta, nodeColor=<rgba 0 0 0 1>}
  n20 = ScalarParameter() {
   defaultValue=1
   desc="0 to 1 animation to visualize delta for Progress.\r\n\r\n0 to 1 animates Progress to the Delta value\r\n1 to 0 animates Progress from the Delta value"
   group="[1] Basics"
   parameterName=AnimateDelta
   sliderMax=1
  }
  NamedRerouteDeclaration(None=@n20) {name=Animate_Delta, nodeColor=<rgba 0 0 0 1>}
  n21 = NamedRerouteUsage()
  n22 = NamedRerouteUsage()
  n23 = NamedRerouteUsage()
  NamedRerouteDeclaration(None=Divide(A=Abs(None=@n21), B=Subtract(A=@n22, B=@n23))) {name=NormalizeDeltaScale, nodeColor=<rgba 0 1 0.9766 1>}
  n24 = ScalarParameter() {defaultValue=1, desc=Set to 1 when progress is between 0 and 1., group="[1] Basics", parameterName=ProgressIsNormalized, sliderMax=1}
  NamedRerouteDeclaration(None=@n24) {name=IsNormalize, nodeColor=<rgba 0 0 0 1>}
  n25 = ScalarParameter() {
   defaultValue=1
   desc="Set to the maximum value of the progress range. \r\n\r\nOnly used when \"ProgressIsNormalized\" = 0"
   group="[1] Basics"
   parameterName=ProgressRangeMax
   sortPriority=33
  }
  NamedRerouteDeclaration(None=@n25) {name=ProgressRangeMax, nodeColor=<rgba 0 0 0 1>}
  n26 = Multiply(
   A=VectorParameter() {group="[2] Colors", parameterName=BackgroundColor, defaultValue=<rgba 0.0242 0.0222 0.0222 1>}.RGB
   B=NamedRerouteUsage()
  )
  n27 = Multiply(A=NamedRerouteUsage(), B=VectorParameter() {group="[2] Colors", parameterName=Color, defaultValue=<rgba 1 1 1 1>}.RGB)
  n28 = Min(A=NamedRerouteUsage(), B=NamedRerouteUsage())
  NamedRerouteDeclaration(None=LinearInterpolate(A=LinearInterpolate(A=@n26, B=@n27, Alpha=@n28), B=NamedRerouteUsage(), Alpha=NamedRerouteUsage())) {name=FinalColor, nodeColor=<rgba 0 0.4375 1 1>}
  NamedRerouteDeclaration(None=MaterialFunctionCall(Size=Constant() {r=1}, Glow Max=Constant() {r=0.2}, Corner Roundness=Constant()) {materialFunction=&MF_FNEUI_Box}.Fill) {name=ShapeUV, nodeColor=<rgba 1 0.6094 0 1>}
  n29 = ScalarParameter() {
   desc="Set to the minimum value of the progress range. \r\n\r\nOnly used when \"ProgressIsNormalized\" = 0"
   group="[1] Basics"
   parameterName=ProgressRangeMin
  }
  NamedRerouteDeclaration(None=@n29) {name=ProgressRangeMin, nodeColor=<rgba 0 0 0 1>}
  n30 = NamedRerouteUsage()
  n31 = NamedRerouteUsage()
  n32 = NamedRerouteUsage()
  NamedRerouteDeclaration(None=Divide(A=Subtract(A=@n30, B=@n31), B=Subtract(A=@n32, B=@n31))) {name=NormalizeProgress, nodeColor=<rgba 0.7344 1 0 1>}
  n33 = Subtract(A=NamedRerouteUsage(), B=NamedRerouteUsage())
  n34 = NamedRerouteUsage()
  n35 = NamedRerouteUsage()
  NamedRerouteDeclaration(None=Divide(A=Subtract(A=@n33, B=@n34), B=Subtract(A=@n35, B=@n34))) {name=NormalizeDeltaOffset, nodeColor=<rgba 0 1 0.9766 1>}
  NamedRerouteDeclaration(None=Multiply(A=NamedRerouteUsage(), B=VectorParameter() {group="[2] Colors", parameterName=DeltaColor, defaultValue=<rgba 0.2917 0.2917 0.2917 1>}.RGB)) {name=DeltaColor, nodeColor=<rgba 0 1 0.2031 1>}
  n36 = NamedRerouteUsage()
  n37 = NamedRerouteUsage()
  n38 = NamedRerouteUsage()
  n39 = NamedRerouteUsage()
  n40 = NamedRerouteUsage()
  NamedRerouteDeclaration(None=If(A=Subtract(A=@n36, B=@n37), B=@n36, "A > B"=@n38, "A == B"=@n38, "A < B"=Min(A=@n39, B=@n40))) {name=DeltaAlpha, nodeColor=<rgba 1 0 0.1172 1>}
  n41 = ScalarParameter() {
   defaultValue=0.5
   desc="Progress Value (should be between 0 and 1 when \"ProgressIsNormalized\" = 1)"
   group="[1] Basics"
   parameterName=Progress
  }
  NamedRerouteDeclaration(None=@n41) {name=Progress, nodeColor=<rgba 0 0 0 1>}
  NamedRerouteUsage()
  NamedRerouteUsage()
 ]
}
