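# Translucent UI material for a linear progress meter that fills horizontally or vertically in either direction, using a normalized or min/max-remapped Progress value.
# Box mask functions shape the background, stroke, bar and end cap, with slant, skew, glow and per-corner rounding.
# Progress translates the bar box, while remappable, rotatable gradients optionally color background, bar and stroke.
# Layers composite through lerps, with optional color-dodge on the end cap.
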Material /FNE_UI/Systems/_Materials/Progress/M_UI_LinearMeter
graph {
 properties {bAutomaticallySetUsageInEditor=false, blendMode=BLEND_Translucent, materialDomain=MD_UI, usedShadingModels=Default Lit}
 nodes [
  n1 = NamedRerouteUsage()
  n2 = NamedRerouteUsage()
  n3 = LinearInterpolate(
   A=@n2
   B=MaterialFunctionCall!(Blend=@n2, Base=NamedRerouteUsage()) {materialFunction=/Engine/Functions/Engine_MaterialFunctions03/Blends/Blend_ColorDodge}.Result
   Alpha=NamedRerouteUsage()
  )
  n4 = NamedRerouteUsage()
  LinearInterpolate(
   A=LinearInterpolate(A=NamedRerouteUsage(), B=LinearInterpolate(A=@n1, B=@n3, Alpha=@n4), Alpha=NamedRerouteUsage()) {constAlpha=1}
   B=NamedRerouteUsage()
   Alpha=NamedRerouteUsage()
  )
  Max(A=Max(A=NamedRerouteUsage(), B=NamedRerouteUsage()), B=NamedRerouteUsage())
  NamedRerouteDeclaration(None=ScalarParameter() {defaultValue=1, group="[6] Stroke", parameterName=StrokeOpacity, sliderMax=1, sortPriority=8}) {name=Scalar_OpacityStroke, nodeColor=<rgba 0 0 0 1>}
  n5 = NamedRerouteUsage()
  n6 = NamedRerouteUsage()
  n7 = MaterialFunctionCall!(
   UVs=MaterialFunctionCall!(
    Input UVs
    Translate=MaterialFunctionCall(X=@n5, Y=@n6) {materialFunction=&MakeFloat2}.Result
   ) {materialFunction=/FNE_UI/Systems/_MaterialFunctions/Transforms/MF_UI_Translate}.UVs raw
   Slant amount=MaterialFunctionCall(X=NamedRerouteUsage(), Y=NamedRerouteUsage()) {materialFunction=&MakeFloat2}.Result
  ) {materialFunction=/FNE_UI/Systems/_MaterialFunctions/Effects/MF_FNEUI_Slant}
  n8 = NamedRerouteUsage()
  n9 = ScalarParameter() {group="[2] Shape", parameterName=SkewRotation, sliderMax=1, sortPriority=14}
  n10 = MaterialFunctionCall!(
   UVs=NamedRerouteDeclaration(None=MaterialFunctionCall!(UVs=@n7.Slanted UVs, Skew amount=@n8, Alpha threshold, Alpha feather, Skew rotation=@n9) {materialFunction=/FNE_UI/Systems/_MaterialFunctions/Distortions/MF_UI_Skew}.Skewed UVs) {name=Transforms, nodeColor=<rgba 0 0 0 1>}
   Size=NamedRerouteUsage()
   Outline Thickness=NamedRerouteUsage()
   Glow Min
   Glow Max=NamedRerouteUsage()
   Corner Roundness=NamedRerouteUsage()
   Keep Aspect Ratio
  ) {materialFunction=/FNE_UI/Systems/_MaterialFunctions/Shapes/MF_FNEUI_Box}
  n11 = NamedRerouteDeclaration(None=@n10.Outline) {name=OutlineMask, nodeColor=<rgba 0 0 0 1>}
  NamedRerouteDeclaration(None=Multiply(A=@n11, B=NamedRerouteUsage())) {name=OutlineFinalColor, nodeColor=<rgba 0 0.1094 1 1>}
  NamedRerouteDeclaration(None=Multiply(A=@n11, B=NamedRerouteUsage())) {name=OutlineAlpha, nodeColor=<rgba 0.5234 0 1 1>}
  n12 = NamedRerouteDeclaration(None=Subtract(A=@n10.Fill, B=@n10.Fill)) {name=BGMask, nodeColor=<rgba 0 0.5078 1 1>}
  NamedRerouteDeclaration(None=Multiply(A=@n12, B=NamedRerouteUsage())) {name=BackgroundAlpha, nodeColor=<rgba 0.8516 1 0 1>}
  NamedRerouteDeclaration(None=ScalarParameter() {group="[2] Shape", parameterName=SlantX, sliderMax=1, sliderMin=-1, sortPriority=11}) {name=Scalar_SlantX, nodeColor=<rgba 0 0 0 1>}
  NamedRerouteDeclaration(None=ScalarParameter() {group="[2] Shape", parameterName=SlantY, sliderMax=1, sliderMin=-1, sortPriority=12}) {name=Scalar_SlantY, nodeColor=<rgba 0 0 0 1>}
  NamedRerouteDeclaration(None=ScalarParameter() {group="[2] Shape", parameterName=OffsetX, sliderMax=0.5, sliderMin=-0.5, sortPriority=5}) {name=Scalar_OffsetX, nodeColor=<rgba 0 0 0 1>}
  n13 = NamedRerouteUsage()
  NamedRerouteDeclaration(None=MaterialFunctionCall(X=@n13, Y=@n13) {materialFunction=&MakeFloat2}.Result) {name=OriginalScale, nodeColor=<rgba 0 0 0 1>}
  n14 = NamedRerouteUsage()
  NamedRerouteDeclaration(None=Multiply(A=ScalarParameter() {group="[2] Shape", parameterName=CornerRoundnessTopLeft, sliderMax=1, sortPriority=2}, B=@n14)) {name=TopLeftCorner, nodeColor=<rgba 0 0 0 1>}
  NamedRerouteDeclaration(None=Multiply(A=ScalarParameter() {group="[2] Shape", parameterName=CornerRoundnessBottomLeft, sliderMax=1, sortPriority=0}, B=@n14)) {name=BottomLeftCorner, nodeColor=<rgba 0 0 0 1>}
  n15 = NamedRerouteUsage()
  n16 = MaterialFunctionCall(Float2=MaterialFunctionCall() {materialFunction=&GetUserInterfaceUV}.Pixel Size) {materialFunction=&BreakOutFloat2Components}
  n17 = NamedRerouteUsage()
  n18 = Subtract(A=@n17, B=Multiply(A=NamedRerouteUsage()) {constB=2})
  n19 = Subtract(A=@n17, B=NamedRerouteUsage())
  n20 = NamedRerouteDeclaration(None=If(
    A=Divide(A=@n16.R, B=@n16.R)
    "A > B"=MaterialFunctionCall(X=@n18, Y=@n19) {materialFunction=&MakeFloat2}.Result
    "A == B"=MaterialFunctionCall(X=@n19, Y=@n19) {materialFunction=&MakeFloat2}.Result
    "A < B"=MaterialFunctionCall(X=@n19, Y=@n18) {materialFunction=&MakeFloat2}.Result
   ) {constB=1}) {name=ShapeScaleBasedOnLongerSide, nodeColor=<rgba 0 0 0 1>}
  n21 = MaterialFunctionCall(UV Channel=MaterialFunctionCall(
    UVs=MaterialFunctionCall!(UVs=@n15, Scale=@n20, Scale center) {materialFunction=/FNE_UI/Systems/_MaterialFunctions/Transforms/MF_UI_Scale}.Result
    Rotation Center=Constant() {r=0.5}
    "Rotation Angle (0-1)"=NamedRerouteUsage()
   ) {materialFunction=&CustomRotator}.Rotated Values) {materialFunction=&LinearGradient}
  n22 = LinearInterpolate(A=@n21.UGradient, B=@n21.UGradient, Alpha=NamedRerouteUsage())
  n23 = VectorParameter() {group="[4] Background", parameterName=BackgroundColor1, sortPriority=0, defaultValue=<rgba 0.018 0.018 0.018 1>}
  n24 = NamedRerouteDeclaration(None=@n23.RGB) {name=Background first color, nodeColor=<rgba 0 1 0.1094 1>}
  n25 = VectorParameter() {group="[4] Background", parameterName=BackgroundColor2, sortPriority=2, defaultValue=<rgba 0 0 0 1>}
  n26 = VectorParameter() {group="[4] Background", parameterName=BackgroundColor3, sortPriority=4, defaultValue=<rgba 0 0 0 1>}
  n27 = VectorParameter() {group="[4] Background", parameterName=BackgroundColor4, sortPriority=6, defaultValue=<rgba 0 0 0 1>}
  n28 = VectorParameter() {group="[4] Background", parameterName=BackgroundColor5, sortPriority=8, defaultValue=<rgba 0 0 0 1>}
  n29 = ScalarParameter() {group="[4] Background", parameterName=BackgroundColor1Position, sortPriority=1}
  n30 = ScalarParameter() {defaultValue=1, group="[4] Background", parameterName=BackgroundColor2Position, sortPriority=3}
  n31 = ScalarParameter() {defaultValue=1, group="[4] Background", parameterName=BackgroundColor3Position, sortPriority=5}
  n32 = ScalarParameter() {defaultValue=1, group="[4] Background", parameterName=BackgroundColor4Position, sortPriority=7}
  n33 = ScalarParameter() {defaultValue=1, group="[4] Background", parameterName=BackgroundColor5Position, sortPriority=9}
  NamedRerouteDeclaration(None=LinearInterpolate(
    A=NamedRerouteUsage()
    B=MaterialFunctionCall!(
     In=@n22
     Value 1=@n24
     Value 2=@n25.RGB
     Value 3=@n26.RGB
     Value 4=@n27.RGB
     Value 5=@n28.RGB
     Pos value 1=@n29
     Pos value 2=@n30
     Pos value 3=@n31
     Pos value 4=@n32
     Pos value 5=@n33
    ) {materialFunction=/FNE_UI/Systems/_MaterialFunctions/Utilities/MF_UI_RemappableGradient}.Gradient 5 Colors
    Alpha=NamedRerouteUsage()
   )) {name=BackgroundColor, nodeColor=<rgba 0 0 0 1>}
  NamedRerouteDeclaration(None=ScalarParameter() {group="[2] Shape", parameterName=OffsetY, sliderMax=0.5, sliderMin=-0.5, sortPriority=6}) {name=Scalar_OffsetY, nodeColor=<rgba 0 0 0 1>}
  n34 = ScalarParameter() {defaultValue=0.1, group="[6] Stroke", parameterName=StrokeThickness, sliderMax=1, sortPriority=9}
  NamedRerouteDeclaration(None=@n34) {name=Scalar_StrokeThickness, nodeColor=<rgba 0 0 0 1>}
  n35 = ScalarParameter() {
   desc="0.0 = Progress fills Horizontally\r\n1.0 = Progress fills Vertically"
   group="[1] Basics"
   parameterName=IsVertical
   sliderMax=1
   sortPriority=0
  }
  NamedRerouteDeclaration(None=@n35) {name=Scalar_Orientation, nodeColor=<rgba 0 0 0 1>}
  n36 = NamedRerouteUsage()
  n37 = Subtract(A=NamedRerouteUsage(), B=@n36)
  n38 = Subtract(A=NamedRerouteUsage(), B=@n36)
  NamedRerouteDeclaration(None=LinearInterpolate(A=Divide(A=@n37, B=@n38), B=NamedRerouteUsage(), Alpha=NamedRerouteUsage())) {name=Scalar_CurrentProgress, nodeColor=<rgba 0 0 0 1>}
  n39 = ScalarParameter() {
   defaultValue=1
   desc="1.0 = Fill progress bar from bottom/left\r\n-1.0 = Fill progress bar from top/right"
   group="[2] Shape"
   parameterName=FillDirection
   sliderMax=1
   sliderMin=-1
   sortPriority=10
  }
  NamedRerouteDeclaration(None=@n39) {name=Scalar_StartPoint, nodeColor=<rgba 0 0 0 1>}
  NamedRerouteDeclaration(None=ScalarParameter() {defaultValue=1, group="[2] Shape", parameterName=Scale, sliderMax=2, sliderMin=0.5, sortPriority=9}) {name=Scalar_Scale, nodeColor=<rgba 0 0 0 1>}
  NamedRerouteDeclaration(None=ScalarParameter() {defaultValue=0.01, group="[3] EndCap", parameterName=EndCapGlow, sliderMax=1, sliderMin=0.01, sortPriority=2}) {name=Scalar_EndCapGlow, nodeColor=<rgba 0 0 0 1>}
  n40 = ScalarParameter() {
   defaultValue=1
   desc=Set to 1 when progress is between 0 and 1.
   group="[1] Basics"
   parameterName=ProgressIsNormalized
   sliderMax=1
   sortPriority=2
  }
  NamedRerouteDeclaration(None=@n40) {name=Scalar_Normalized, nodeColor=<rgba 0 0 0 1>}
  n41 = MaterialFunctionCall(Float2=NamedRerouteUsage()) {materialFunction=&BreakOutFloat2Components}
  n42 = MaterialFunctionCall(Float2=MaterialFunctionCall() {materialFunction=&GetUserInterfaceUV}.Pixel Size) {materialFunction=&BreakOutFloat2Components}
  n43 = Multiply(A=@n41.G, B=@n42.G)
  n44 = Multiply(A=@n41.R, B=@n42.R)
  n45 = Divide(A=NamedRerouteUsage())
  n46 = Divide(A=Multiply(A=@n43, B=@n45), B=@n44)
  n47 = NamedRerouteUsage()
  n48 = Divide(A=Multiply(A=@n44, B=@n45), B=@n43)
  n49 = NamedRerouteUsage()
  NamedRerouteDeclaration(None=If(
    A=Divide(A=@n44, B=@n43)
    "A > B"=LinearInterpolate(A=@n46, B=@n45, Alpha=@n47)
    "A == B"=@n45
    "A < B"=LinearInterpolate(A=@n45, B=@n48, Alpha=@n49)
   ) {constB=1}) {name=OutlineValueForLongerSide, nodeColor=<rgba 0 0 0 1>}
  NamedRerouteDeclaration(None=Multiply(A=NamedRerouteUsage(), B=@n34)) {name=StrokeThicknessBasedOnScale, nodeColor=<rgba 0 0 0 1>}
  n50 = NamedRerouteUsage()
  n51 = NamedRerouteUsage()
  n52 = MaterialFunctionCall(UV Channel=MaterialFunctionCall(
    UVs=MaterialFunctionCall(UVs=@n50, Scale=@n51) {materialFunction=&MF_UI_Scale}.Result
    Rotation Center=Constant() {r=0.5}
    "Rotation Angle (0-1)"=NamedRerouteUsage()
   ) {materialFunction=&CustomRotator}.Rotated Values) {materialFunction=&LinearGradient}
  n53 = LinearInterpolate(A=@n52.UGradient, B=@n52.UGradient, Alpha=NamedRerouteUsage())
  n54 = NamedRerouteDeclaration(None=VectorParameter() {group="[5] Fill", parameterName=BarColor1, sortPriority=0, defaultValue=<rgba 1 1 1 1>}.RGB) {name=Bar first color, nodeColor=<rgba 0.1484 1 0 1>}
  n55 = VectorParameter() {group="[5] Fill", parameterName=BarColor2, sortPriority=2, defaultValue=<rgba 1 1 1 1>}
  n56 = VectorParameter() {group="[5] Fill", parameterName=BarColor3, sortPriority=4, defaultValue=<rgba 1 1 1 1>}
  n57 = VectorParameter() {group="[5] Fill", parameterName=BarColor4, sortPriority=6, defaultValue=<rgba 1 1 1 1>}
  n58 = VectorParameter() {group="[5] Fill", parameterName=BarColor5, sortPriority=8, defaultValue=<rgba 1 1 1 1>}
  n59 = ScalarParameter() {group="[5] Fill", parameterName=BarColor1Position, sortPriority=1}
  n60 = ScalarParameter() {defaultValue=0.5, group="[5] Fill", parameterName=BarColor2Position, sortPriority=3}
  n61 = ScalarParameter() {defaultValue=1, group="[5] Fill", parameterName=BarColor3Position, sortPriority=5}
  n62 = ScalarParameter() {defaultValue=1, group="[5] Fill", parameterName=BarColor4Position, sortPriority=7}
  n63 = ScalarParameter() {defaultValue=1, group="[5] Fill", parameterName=BarColor5Position, sortPriority=9}
  NamedRerouteDeclaration(None=LinearInterpolate(
    A=NamedRerouteUsage()
    B=MaterialFunctionCall(
     In=@n53
     Value 1=@n54
     Value 2=@n55.RGB
     Value 3=@n56.RGB
     Value 4=@n57.RGB
     Value 5=@n58.RGB
     Pos value 1=@n59
     Pos value 2=@n60
     Pos value 3=@n61
     Pos value 4=@n62
     Pos value 5=@n63
    ) {materialFunction=&MF_UI_RemappableGradient}.Gradient 5 Colors
    Alpha=NamedRerouteUsage()
   )) {name=BaseColor, nodeColor=<rgba 0 0 0 1>}
  NamedRerouteDeclaration(None=ScalarParameter() {defaultValue=0.25, group="[4] Background", parameterName=BackgroundGradientRotation, sliderMax=1, sortPriority=11}) {name=BackgroundGradientRotation, nodeColor=<rgba 0 0 0 1>}
  NamedRerouteDeclaration(None=Subtract(A=NamedRerouteUsage(), B=NamedRerouteUsage()) {constB=0}) {name=CurrentProgressMinusGapThickness, nodeColor=<rgba 0 1 0.0391 1>}
  n64 = NamedRerouteUsage()
  NamedRerouteDeclaration(None=Multiply(A=ScalarParameter() {group="[2] Shape", parameterName=CornerRoundnessTopRight, sliderMax=1, sortPriority=3}, B=@n64)) {name=TopRightCorner, nodeColor=<rgba 0 0 0 1>}
  NamedRerouteDeclaration(None=Multiply(A=ScalarParameter() {group="[2] Shape", parameterName=CornerRoundnessBottomRight, sliderMax=1, sortPriority=1}, B=@n64)) {name=BottomRightCorner, nodeColor=<rgba 0 0 0 1>}
  NamedRerouteDeclaration(None=ScalarParameter() {group="[2] Shape", parameterName=RoundedProgress, sliderMax=1, sortPriority=8}) {name=Scalar_RoundedProgress, nodeColor=<rgba 0 0 0 1>}
  NamedRerouteDeclaration(None=MaterialFunctionCall(X=NamedRerouteUsage(), Y=NamedRerouteUsage(), Z=NamedRerouteUsage(), A=NamedRerouteUsage()) {materialFunction=&MakeFloat4}.Result) {name=Corners, nodeColor=<rgba 0 0 0 1>}
  NamedRerouteDeclaration(None=Multiply(A=@n12, B=NamedRerouteUsage())) {name=BackgroundFinalColor, nodeColor=<rgba 0 1 0.1797 1>}
  NamedRerouteDeclaration(None=ScalarParameter() {group="[2] Shape", parameterName=SkewAmount, sliderMax=1, sortPriority=13}) {name=Scalar_SkewAmount, nodeColor=<rgba 0 0 0 1>}
  NamedRerouteDeclaration(None=ScalarParameter() {defaultValue=1, group="[5] Fill", parameterName=BarOpacity, sliderMax=1, sortPriority=12}) {name=Scalar_OpacityBar, nodeColor=<rgba 0 0 0 1>}
  NamedRerouteDeclaration(None=ScalarParameter() {group="[6] Stroke", parameterName=StrokeGradientRotation, sliderMax=1, sortPriority=7}) {name=OutlineGradientRotation, nodeColor=<rgba 0 0 0 1>}
  NamedRerouteDeclaration(None=ScalarParameter() {group="[5] Fill", parameterName=BarGradientOn, sliderMax=1, sortPriority=10}) {name=Scalar_BarGradientOn, nodeColor=<rgba 0 0 0 1>}
  NamedRerouteDeclaration(None=ScalarParameter() {group="[4] Background", parameterName=BackgroundGradientOn, sliderMax=1, sortPriority=10}) {name=Scalar_BackgroundGradientOn, nodeColor=<rgba 0 0 0 1>}
  n65 = NamedRerouteUsage()
  NamedRerouteDeclaration(None=Multiply(
    A=ScalarParameter() {group="[3] EndCap", parameterName=EndCapThickness, sliderMax=0.8, sortPriority=5}
    B=MaterialFunctionCall(Float2=@n65) {materialFunction=&BreakOutFloat2Components}.R
   )) {name=Scalar_GapThickness, nodeColor=<rgba 0 0 0 1>}
  NamedRerouteDeclaration(None=ScalarParameter() {group="[6] Stroke", parameterName=StrokeGradientOn, sliderMax=1, sortPriority=6}) {name=Scalar_StrokeGradientOn, nodeColor=<rgba 0 0 0 1>}
  NamedRerouteDeclaration(None=ScalarParameter() {group="[5] Fill", parameterName=BarGradientRotation, sliderMax=1, sortPriority=11}) {name=BarGradientRotation, nodeColor=<rgba 0 0 0 1>}
  NamedRerouteDeclaration(None=ScalarParameter() {group="[3] EndCap", parameterName=EndCapSlant, sliderMax=0.1, sliderMin=-0.1, sortPriority=4}) {name=Scalar_EndCapSlant, nodeColor=<rgba 0 0 0 1>}
  n66 = NamedRerouteUsage()
  n67 = Constant() {r=1}
  n68 = NamedRerouteUsage()
  n69 = NamedRerouteUsage()
  n70 = NamedRerouteUsage()
  n71 = NamedRerouteUsage()
  n72 = NamedRerouteUsage()
  n73 = MaterialFunctionCall(X=LinearInterpolate(A=@n71, Alpha=@n72) {constB=0}, Y=LinearInterpolate(B=@n71, Alpha=@n72)) {materialFunction=&MakeFloat2}
  n74 = MaterialFunctionCall(UV Channel=MaterialFunctionCall(
    UVs=MaterialFunctionCall(UVs=@n70, Slant amount=@n73.Result) {materialFunction=&MF_FNEUI_Slant}.Slanted UVs
    Scale=NamedRerouteUsage()
   ) {materialFunction=&MF_UI_Scale}.Result) {materialFunction=&LinearGradient}
  n75 = LinearInterpolate(
   A=If(A=@n69, "A > B"=@n74.UGradient, "A == B"=@n74.UGradient, "A < B"=OneMinus(None=@n74.UGradient))
   B=If(A=@n69, "A > B"=OneMinus(None=@n74.VGradient), "A == B"=@n74.VGradient, "A < B"=@n74.VGradient)
   Alpha=NamedRerouteUsage()
  )
  NamedRerouteDeclaration(None=MaterialFunctionCall!(
    Gradient mask=@n75
    Value 1=Subtract(B=@n66)
    Value 2=@n67
    Value 3=@n67
    Value 4=@n67
    Value 5=@n67
    Pos value 1=NamedRerouteUsage()
    Pos value 2=@n68
    Pos value 3=@n68
    Pos value 4=@n68
    Pos value 5=@n68
   ) {materialFunction=/FNE_UI/Systems/_MaterialFunctions/Utilities/MF_UI_RemappableGradient_Float}.Gradient 3 Values) {name=EndCapGradientMask, nodeColor=<rgba 0 0.3672 1 1>}
  NamedRerouteDeclaration(None=ScalarParameter() {defaultValue=1, group="[4] Background", parameterName=BackgroundOpacity, sliderMax=1, sortPriority=12}) {name=Scalar_OpacityBackground, nodeColor=<rgba 0 0 0 1>}
  n76 = NamedRerouteUsage()
  n77 = NamedRerouteUsage()
  n78 = NamedRerouteUsage()
  n79 = NamedRerouteUsage()
  n80 = MaterialFunctionCall(X=LinearInterpolate(A=@n78, Alpha=@n79) {constB=0}, Y=LinearInterpolate(B=@n78, Alpha=@n79)) {materialFunction=&MakeFloat2}
  n81 = Multiply(A=NamedRerouteUsage(), B=Multiply(A=NamedRerouteUsage()) {constB=-1})
  n82 = Constant()
  n83 = NamedRerouteUsage()
  n84 = LinearInterpolate(A=@n81, B=@n82, Alpha=@n83)
  n85 = LinearInterpolate(A=@n82, B=@n81, Alpha=@n83)
  n86 = Constant()
  n87 = NamedRerouteUsage()
  n88 = NamedRerouteUsage()
  n89 = NamedRerouteUsage()
  n90 = NamedRerouteUsage()
  n91 = NamedRerouteUsage()
  n92 = NamedRerouteUsage()
  n93 = MaterialFunctionCall(
   UVs=MaterialFunctionCall(
    Input UVs=MaterialFunctionCall(UVs=@n77, Slant amount=@n80.Result) {materialFunction=&MF_FNEUI_Slant}.Slanted UVs
    Translate=MaterialFunctionCall(X=@n84, Y=@n85) {materialFunction=&MakeFloat2}.Result
   ) {materialFunction=&MF_UI_Translate}.UVs raw
   Size=NamedRerouteUsage()
   Outline Thickness=NamedRerouteUsage()
   Glow Min=@n86
   Glow Max=@n86
   Corner Roundness=MaterialFunctionCall(X=Multiply(A=@n87, B=@n88), Y=Multiply(A=@n89, B=@n88), Z=Multiply(A=@n90, B=@n91), A=Multiply(A=@n92, B=@n91)) {materialFunction=&MakeFloat4}.Result
  ) {materialFunction=&MF_FNEUI_Box}
  n94 = Subtract(A=@n93.Fill, B=@n93.Fill)
  n95 = NamedRerouteUsage()
  n96 = Subtract(A=NamedRerouteUsage(), B=Min(A=Multiply(A=@n94, B=@n95), B=NamedRerouteUsage()))
  n97 = NamedRerouteUsage()
  NamedRerouteDeclaration(None=Min(A=@n76, B=Multiply(A=@n96, B=@n97))) {name=EndCapMask, nodeColor=<rgba 0.7344 1 0 1>}
  n98 = VectorParameter() {group="[3] EndCap", parameterName=EndCapColor, sortPriority=0, defaultValue=<rgba 0 1 1 1>}
  n99 = NamedRerouteUsage()
  NamedRerouteDeclaration(None=MaterialFunctionCall(
    In=@n75
    Value 1=@n98.RGB
    Value 2=@n98.RGB
    Value 3=@n98.RGB
    Value 4=@n98.RGB
    Value 5=@n98.RGB
    Pos value 1=NamedRerouteUsage()
    Pos value 2=@n99
    Pos value 3=@n99
    Pos value 4=@n99
    Pos value 5=@n99
   ) {materialFunction=&MF_UI_RemappableGradient}.Gradient 3 Colors) {name=EndCapColor, nodeColor=<rgba 0 0 0 1>}
  n100 = NamedRerouteUsage()
  NamedRerouteDeclaration(None=Min(A=Multiply(A=@n96, B=@n100), B=@n76)) {name=EndCapFinalColor, nodeColor=<rgba 1 0.1172 0 1>}
  NamedRerouteDeclaration(None=ScalarParameter() {group="[3] EndCap", parameterName=EndCapOpacity, sliderMax=1, sortPriority=3}) {name=Scalar_EndCapOpacity, nodeColor=<rgba 0 0 0 1>}
  NamedRerouteDeclaration(None=ScalarParameter() {group="[3] EndCap", parameterName=EndCapColorDodge, sliderMax=1, sortPriority=1}) {name=Scalar_EndGapEffect, nodeColor=<rgba 0 0 0 1>}
  NamedRerouteDeclaration(None=ScalarParameter() {group="[2] Shape", parameterName=GlowMax, sliderMax=0.2, sortPriority=4}) {name=Scalar_GlowMax, nodeColor=<rgba 0 0 0 1>}
  n101 = ScalarParameter() {
   desc="Set to the minimum value of the progress range. \r\n\r\nOnly used when \"ProgressIsNormalized\" = 0"
   group="[1] Basics"
   parameterName=ProgressRangeMin
   sortPriority=3
  }
  NamedRerouteDeclaration(None=@n101) {name=Scalar_ProgressMin, nodeColor=<rgba 0 0 0 1>}
  n102 = ScalarParameter() {
   defaultValue=1
   desc="Set to the maximum value of the progress range. \r\n\r\nOnly used when \"ProgressIsNormalized\" = 0"
   group="[1] Basics"
   parameterName=ProgressRangeMax
  }
  NamedRerouteDeclaration(None=@n102) {name=Scalar_ProgressMax, nodeColor=<rgba 0 0 0 1>}
  n103 = NamedRerouteUsage()
  n104 = NamedRerouteUsage()
  n105 = Multiply(A=NamedRerouteUsage(), B=Subtract(A=NamedRerouteUsage(), B=Multiply(A=@n104) {constB=2}))
  n106 = NamedRerouteUsage()
  n107 = NamedRerouteUsage()
  n108 = Subtract(A=Multiply(A=Add(A=@n105, B=@n106), B=NamedRerouteUsage()), B=Multiply(A=Subtract(A=@n107, B=@n104), B=NamedRerouteUsage()))
  n109 = Constant()
  n110 = LinearInterpolate(
   A=MaterialFunctionCall(X=@n108, Y=@n109) {materialFunction=&MakeFloat2}.Result
   B=MaterialFunctionCall(X=@n109, Y=@n108) {materialFunction=&MakeFloat2}.Result
   Alpha=NamedRerouteUsage()
  )
  n111 = NamedRerouteUsage()
  n112 = NamedRerouteUsage()
  n113 = NamedRerouteUsage()
  n114 = NamedRerouteUsage()
  n115 = NamedRerouteUsage()
  n116 = NamedRerouteUsage()
  n117 = MaterialFunctionCall(
   UVs=NamedRerouteDeclaration(None=MaterialFunctionCall(Input UVs=@n103, Translate=@n110) {materialFunction=&MF_UI_Translate}.UVs raw) {name=TransformUV, nodeColor=<rgba 0 0 0 1>}
   Size=NamedRerouteUsage()
   Outline Thickness=NamedRerouteUsage()
   Glow Min=Constant()
   Glow Max=NamedRerouteUsage()
   Corner Roundness=MaterialFunctionCall(X=Multiply(A=@n111, B=@n112), Y=Multiply(A=@n113, B=@n112), Z=Multiply(A=@n114, B=@n115), A=Multiply(A=@n116, B=@n115)) {materialFunction=&MakeFloat4}.Result
  ) {materialFunction=&MF_FNEUI_Box}
  n118 = NamedRerouteDeclaration(None=Subtract(A=@n117.Fill, B=@n117.Fill)) {name=BarMask, nodeColor=<rgba 0 0 0 1>}
  NamedRerouteDeclaration(None=Multiply(A=@n118, B=NamedRerouteUsage())) {name=BarFinalColor, nodeColor=<rgba 0 0.4609 1 1>}
  n119 = ScalarParameter() {
   desc="Progress Value (should be between 0 and 1 when \"ProgressIsNormalized\" = 1)"
   group="[1] Basics"
   parameterName=Progress
   sortPriority=1
  }
  NamedRerouteDeclaration(None=@n119) {name=Scalar_Progress, nodeColor=<rgba 0 0 0 1>}
  n120 = NamedRerouteUsage()
  NamedRerouteDeclaration(None=Min(A=Multiply(A=@n118, B=@n120), B=NamedRerouteUsage())) {name=BarAlpha, nodeColor=<rgba 0 1 0.9766 1>}
  n121 = NamedRerouteUsage()
  n122 = NamedRerouteUsage()
  n123 = MaterialFunctionCall(UV Channel=MaterialFunctionCall(
    UVs=MaterialFunctionCall(UVs=@n121, Scale=@n122) {materialFunction=&MF_UI_Scale}.Result
    Rotation Center=Constant() {r=0.5}
    "Rotation Angle (0-1)"=NamedRerouteUsage()
   ) {materialFunction=&CustomRotator}.Rotated Values) {materialFunction=&LinearGradient}
  n124 = LinearInterpolate(A=@n123.UGradient, B=@n123.UGradient, Alpha=NamedRerouteUsage())
  n125 = NamedRerouteDeclaration(None=VectorParameter() {group="[6] Stroke", parameterName=StrokeColor1, sortPriority=0, defaultValue=<rgba 1 1 1 1>}.RGB) {name=StrokeFirstColor, nodeColor=<rgba 0.1719 0 1 1>}
  n126 = VectorParameter() {group="[6] Stroke", parameterName=StrokeColor2, sortPriority=2, defaultValue=<rgba 1 1 1 1>}
  n127 = VectorParameter() {group="[6] Stroke", parameterName=StrokeColor3, sortPriority=4, defaultValue=<rgba 1 1 1 1>}
  n128 = ScalarParameter() {group="[6] Stroke", parameterName=StrokeColor1Position, sortPriority=1}
  n129 = ScalarParameter() {defaultValue=0.5, group="[6] Stroke", parameterName=StrokeColor2Position, sortPriority=3}
  n130 = ScalarParameter() {defaultValue=1, group="[6] Stroke", parameterName=StrokeColor3Position, sortPriority=5}
  NamedRerouteDeclaration(None=LinearInterpolate(
    A=NamedRerouteUsage()
    B=MaterialFunctionCall(In=@n124, Value 1=@n125, Value 2=@n126.RGB, Value 3=@n127.RGB, Pos value 1=@n128, Pos value 2=@n129, Pos value 3=@n130) {materialFunction=&MF_UI_RemappableGradient}.Gradient 3 Colors
    Alpha=NamedRerouteUsage()
   )) {name=OutlineColor, nodeColor=<rgba 0 0 0 1>}
 ]
}
