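# UI material for a radial progress meter: a rotatable arc with optional segments, color gradient and stroke.
# SDF circle functions shape the ring and stroke, while polar coordinates produce the angular mask between begin and end positions.
# A ping-pong function cuts segments and gaps, with stroke spacing compensated for ring size; normalized or ranged progress fills smoothly or snaps per segment.
# Named reroutes link the stages, and the three-color gradient blends between angular and radial mapping.
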Material /FNE_UI/Systems/_Materials/Progress/M_UI_RadialMeter
graph {
 properties {bAutomaticallySetUsageInEditor=false, blendMode=BLEND_Translucent, materialDomain=MD_UI, usedShadingModels=Default Lit}
 nodes [
  n1 = VectorParameter() {group="[5] Stroke", parameterName=StrokeColor, sortPriority=1, defaultValue=<rgba 1 1 1 1>}
  n2 = MaterialFunctionCall!(
   UV
   Size=NamedRerouteUsage()
   GlowMin
   GlowMax=NamedRerouteUsage()
   Stroke thickness=NamedRerouteUsage()
   Is stroke thickness relative to size
  ) {materialFunction=/FNE_UI/Systems/_MaterialFunctions/SDFs/MF_UI_SDF_Circle}
  n3 = Min(A=NamedRerouteUsage(), B=@n2.Outline)
  n4 = Min(A=@n3, B=NamedRerouteUsage())
  LinearInterpolate(
   A=LinearInterpolate(
    A=VectorParameter() {group="[5] Stroke", parameterName=StrokeEmptyColor, sortPriority=0, defaultValue=<rgba 0 0 0 1>}.RGB
    B=Multiply(A=@n1.RGB, B=@n3)
    Alpha=@n4
   )
   B=NamedRerouteUsage()
   Alpha=NamedRerouteUsage()
  )
  n5 = NamedRerouteUsage()
  Max(A=@n4, B=Min(A=NamedRerouteUsage(), B=Multiply(A=@n2.Outline, B=@n5)))
  NamedRerouteDeclaration(None=ScalarParameter() {defaultValue=0.99, group="[2] Shape", parameterName=Scale, sliderMax=1, sortPriority=4}) {name=Scalar_Scale, nodeColor=<rgba 0 0 0 1>}
  NamedRerouteDeclaration(None=Subtract(
    A=ScalarParameter() {defaultValue=0.2, group="[2] Shape", parameterName=Thickness, sliderMax=1, sliderMin=0.1, sortPriority=5}
    B=NamedRerouteUsage()
   )) {name=Scalar_Thickness, nodeColor=<rgba 0 0 0 1>}
  NamedRerouteDeclaration(None=VectorParameter() {group="[4] Color", parameterName=Color3, sortPriority=6, defaultValue=<rgba 1 1 1 1>}.RGB) {name=ThirdColor, nodeColor=<rgba 0 0 0 1>}
  NamedRerouteDeclaration(None=ScalarParameter() {defaultValue=1, group="[4] Color", parameterName=RadialGradient, sliderMax=1, sortPriority=9}) {name=Scalar_RadialGradient, nodeColor=<rgba 0 0 0 1>}
  NamedRerouteDeclaration(None=ScalarParameter() {group="[4] Color", parameterName=RadialGradientSize, sliderMax=1, sliderMin=-2, sortPriority=10}) {name=Scalar_RadialGradientSize, nodeColor=<rgba 0 0 0 1>}
  n6 = NamedRerouteUsage()
  n7 = Subtract(A=NamedRerouteUsage(), B=@n6)
  n8 = Subtract(A=NamedRerouteUsage(), B=@n6)
  NamedRerouteDeclaration(None=LinearInterpolate(A=Divide(A=@n7, B=@n8), B=NamedRerouteUsage(), Alpha=NamedRerouteUsage())) {name=Scalar_CurrentProgress, nodeColor=<rgba 0 0 0 1>}
  NamedRerouteDeclaration(None=ScalarParameter() {defaultValue=0.5, group="[4] Color", parameterName=EmptyOpacity, sliderMax=1, sortPriority=1}) {name=Scalar_EmptyStateOpacity, nodeColor=<rgba 0 0 0 1>}
  n9 = Constant() {r=-0.002}
  n10 = NamedRerouteUsage()
  n11 = NamedRerouteUsage()
  n12 = LinearInterpolate(
   A=@n11
   B=MaterialFunctionCall!(Gradient=@n11, Num steps=NamedRerouteUsage()) {materialFunction=/FNE_UI/Systems/_MaterialFunctions/Utilities/MF_UI_SteppedGradient}.Output keep steps difference
   Alpha=NamedRerouteUsage()
  )
  n13 = SmoothStep(Min=Clamp(None=Add(A=@n9, B=@n10)), Max=@n10, Value=@n12)
  n14 = SmoothStep(Max=Multiply() {constA=-0.002, constB=-1}, Value=@n12)
  n15 = NamedRerouteUsage()
  n16 = LinearInterpolate(A=Subtract(A=OneMinus(None=@n13), B=OneMinus(None=@n14)), B=OneMinus(None=Step(Y=@n15, X=@n12)), Alpha=NamedRerouteUsage())
  n17 = If(A=NamedRerouteUsage(), "A > B"=@n16, "A == B"=Constant(), "A < B"=@n16)
  NamedRerouteDeclaration(None=If(A=NamedRerouteUsage(), "A > B"=@n17, "A == B"=Constant() {r=1}, "A < B"=@n17) {constB=1}) {name=ProgressMask, nodeColor=<rgba 0.2188 1 0 1>}
  NamedRerouteDeclaration(None=VectorParameter() {group="[4] Color", parameterName=EmptyColor, sortPriority=0, defaultValue=<rgba 0 0 0 1>}.RGB) {name=EmptyStateColor, nodeColor=<rgba 0 0 0 1>}
  n18 = ScalarParameter() {
   defaultValue=1
   desc=Set to 1 when progress is between 0 and 1.
   group="[1] Basics"
   parameterName=ProgressIsNormalized
   sliderMax=1
   sortPriority=1
  }
  NamedRerouteDeclaration(None=@n18) {name=Scalar_Normalized, nodeColor=<rgba 0 0 0 1>}
  n19 = ScalarParameter() {
   desc="Determines if progress fill should snap to each segment or fill smoothly.\r\n\r\n0.0 = No snapping\r\n1.0 = Snapping"
   group="[3] Segments"
   parameterName=Snapping
   sliderMax=1
   sortPriority=3
  }
  NamedRerouteDeclaration(None=@n19) {name=Scalar_ProgressSnapping, nodeColor=<rgba 0 0 0 1>}
  NamedRerouteDeclaration(None=ScalarParameter() {group="[2] Shape", parameterName=PositionBegin, sliderMax=1, sortPriority=1}) {name=Scalar_PositionBegin, nodeColor=<rgba 0 0 0 1>}
  NamedRerouteDeclaration(None=ScalarParameter() {defaultValue=0.009, group="[2] Shape", parameterName=EdgeSoftness, sliderMax=0.2, sortPriority=0}) {name=Scalar_GlowMax, nodeColor=<rgba 0 0 0 1>}
  NamedRerouteDeclaration(None=ScalarParameter() {defaultValue=1, group="[2] Shape", parameterName=PositionEnd, sliderMax=1, sortPriority=2}) {name=Scalar_PositionEnd, nodeColor=<rgba 0 0 0 1>}
  n20 = ScalarParameter() {group="[5] Stroke", parameterName=StrokeThickness, sliderMax=1, sortPriority=2}
  NamedRerouteDeclaration(None=Multiply(A=NamedRerouteDeclaration(None=@n20) {name=StrokeThicknessOriginal, nodeColor=<rgba 0.6641 0 1 1>}, B=NamedRerouteUsage())) {name=Scalar_StrokeThickness, nodeColor=<rgba 0 0 0 1>}
  n21 = NamedRerouteUsage()
  n22 = SmoothStep(
   Min=@n21
   Max=Add(A=@n21, B=NamedRerouteUsage())
   Value=MaterialFunctionCall!(Value=NamedRerouteUsage(), Scale=NamedRerouteUsage(), Height) {materialFunction=/FNE_UI/Systems/_MaterialFunctions/Time/MF_FNEUI_PingPong}.Result
  )
  NamedRerouteDeclaration(None=If(A=NamedRerouteUsage(), "A > B"=@n22, "A == B"=Constant() {r=1}, "A < B"=@n22) {constB=1}) {name=RadialSegmentsStroke, nodeColor=<rgba 0.9688 0 1 1>}
  n23 = NamedRerouteUsage()
  n24 = Multiply(
   A=MaterialFunctionCall(Multiplier=Constant() {r=2}) {materialFunction=&Pi}.Result
   B=NamedRerouteUsage()
  )
  n25 = Multiply(A=@n24, B=@n23)
  n26 = NamedRerouteUsage()
  n27 = NamedRerouteUsage()
  n28 = Divide(A=Multiply(A=Subtract(A=@n24, B=@n25), B=Subtract(A=@n26, B=@n27)), B=NamedRerouteUsage())
  n29 = Multiply(A=NamedRerouteUsage(), B=NamedRerouteUsage())
  n30 = Add(A=@n23, B=Divide(A=NamedRerouteUsage(), B=Divide(A=@n28, B=@n29))) {constB=0}
  n31 = Subtract(A=NamedRerouteUsage(), B=NamedRerouteUsage())
  n32 = If(A=NamedRerouteUsage(), "A > B"=@n30, "A == B"=Constant(), "A < B"=@n30)
  NamedRerouteDeclaration(None=If(A=NamedRerouteUsage(), "A > B"=@n30, "A == B"=If(A=@n31, "A > B"=@n30, "A == B"=@n32, "A < B"=@n30) {constB=1}, "A < B"=@n30) {constB=1}) {name=SegmentStrokeSpacing, nodeColor=<rgba 0 0 0 1>}
  n33 = NamedRerouteUsage()
  n34 = MaterialFunctionCall!(
   Vector or UVs=MaterialFunctionCall(UVs=TextureCoordinate(), Rotation Center=Constant() {r=0.5}, "Rotation Angle (0-1)"=Multiply(A=@n33) {constB=-1}) {materialFunction=&CustomRotator}.Rotated Values
   Swizzle Coordinate Output
  ) {materialFunction=/FNE_UI/Systems/_MaterialFunctions/Distortions/MF_UI_PolarCoordinates}
  n35 = Constant()
  n36 = NamedRerouteUsage()
  n37 = NamedRerouteUsage()
  n38 = Divide(A=Subtract(A=@n37, B=NamedRerouteUsage()))
  n39 = NamedRerouteUsage()
  NamedRerouteDeclaration(None=MaterialFunctionCall!(
    In=OneMinus(None=@n34.Angle gradient)
    Value 1=@n35
    Value 2=@n35
    Value 3=Constant() {r=0.5}
    Value 4=Constant() {r=1}
    Value 5=Constant()
    Pos value 1=@n36
    Pos value 2=@n36
    Pos value 3=Subtract(A=@n37, B=@n38)
    Pos value 4=@n39
    Pos value 5=@n39
   ) {materialFunction=/FNE_UI/Systems/_MaterialFunctions/Utilities/MF_UI_RemappableGradient}.Gradient 5 Colors) {name=ArcMask, nodeColor=<rgba 0 1 0.7422 1>}
  n40 = Multiply(A=NamedRerouteUsage(), B=NamedRerouteUsage())
  n41 = Subtract(A=NamedRerouteUsage(), B=NamedRerouteUsage())
  n42 = Constant()
  n43 = MaterialFunctionCall(
   Size=Subtract(A=NamedRerouteUsage(), B=Divide(A=@n40))
   GlowMax=If(A=@n41, "A > B"=NamedRerouteUsage(), "A == B"=@n42, "A < B"=@n42)
   Stroke thickness=@n41
  ) {materialFunction=&MF_UI_SDF_Circle}
  n44 = Min(A=Min(A=NamedRerouteUsage(), B=@n43.Outline), B=NamedRerouteUsage())
  n45 = NamedRerouteUsage()
  n46 = Multiply(A=@n43.Outline, B=NamedRerouteUsage())
  NamedRerouteDeclaration(None=Max(A=@n44, B=Min(A=@n45, B=@n46))) {name=StrokeAlpha, nodeColor=<rgba 1 0.3281 0 1>}
  n47 = LinearInterpolate(A=NamedRerouteUsage(), B=NamedRerouteUsage(), Alpha=NamedRerouteUsage())
  NamedRerouteDeclaration(None=LinearInterpolate(A=NamedRerouteUsage(), B=Multiply(A=@n47, B=@n44), Alpha=@n44)) {name=StrokeFill, nodeColor=<rgba 0.0078 0 1 1>}
  n48 = ScalarParameter() {
   defaultValue=0.5
   desc="Progress Value (should be between 0 and 1 when \"ProgressIsNormalized\" = 1)"
   group="[1] Basics"
   parameterName=Progress
   sortPriority=0
  }
  NamedRerouteDeclaration(None=@n48) {name=Scalar_Progress, nodeColor=<rgba 0 0 0 1>}
  n49 = NamedRerouteUsage()
  n50 = SmoothStep(
   Min=@n49
   Max=Add(A=@n49, B=NamedRerouteUsage())
   Value=MaterialFunctionCall(Value=NamedRerouteUsage(), Scale=NamedRerouteUsage()) {materialFunction=&MF_FNEUI_PingPong}.Result
  )
  n51 = Subtract(A=NamedRerouteUsage(), B=NamedRerouteUsage())
  n52 = Constant() {r=1}
  NamedRerouteDeclaration(None=If(A=NamedRerouteUsage(), "A > B"=@n50, "A == B"=If(A=@n51, "A > B"=@n50, "A == B"=@n52, "A < B"=@n50) {constB=1}, "A < B"=@n50) {constB=1}) {name=RadialSegments, nodeColor=<rgba 0.6641 1 0 1>}
  NamedRerouteDeclaration(None=ScalarParameter() {defaultValue=10, group="[3] Segments", parameterName=SegmentCount, sortPriority=0}) {name=Scalar_SegmentCount, nodeColor=<rgba 0 0 0 1>}
  NamedRerouteDeclaration(None=ScalarParameter() {group="[2] Shape", parameterName=Rotation, sliderMax=1, sortPriority=3}) {name=Scalar_Rotation, nodeColor=<rgba 0 0 0 1>}
  n53 = ScalarParameter() {
   desc="Set to the minimum value of the progress range. \r\n\r\nOnly used when \"ProgressIsNormalized\" = 0"
   group="[1] Basics"
   parameterName=ProgressRangeMin
  }
  NamedRerouteDeclaration(None=@n53) {name=Scalar_ProgressMin, nodeColor=<rgba 0 0 0 1>}
  n54 = ScalarParameter() {
   defaultValue=1
   desc="Set to the maximum value of the progress range.\r\n\r\nOnly used when \"ProgressIsNormalized\" = 0"
   group="[1] Basics"
   parameterName=ProgressRangeMax
   sortPriority=33
  }
  NamedRerouteDeclaration(None=@n54) {name=Scalar_ProgressMax, nodeColor=<rgba 0 0 0 1>}
  n55 = ScalarParameter() {defaultValue=0.1, group="[3] Segments", parameterName=SegmentGap, sliderMax=1, sortPriority=1}
  n56 = Subtract(A=NamedRerouteUsage(), B=NamedRerouteUsage())
  n57 = Constant()
  NamedRerouteDeclaration(None=If(A=NamedRerouteUsage(), "A > B"=@n55, "A == B"=If(A=@n56, "A > B"=@n55, "A == B"=@n57, "A < B"=@n55) {constB=1}, "A < B"=@n55) {constB=1}) {name=SegmentSpacing, nodeColor=<rgba 0 0 0 1>}
  n58 = ScalarParameter() {defaultValue=0.02, group="[3] Segments", parameterName=SegmentEdgeSoftness, sliderMax=1, sortPriority=2}
  n59 = Subtract(A=NamedRerouteUsage(), B=NamedRerouteUsage())
  n60 = ScalarParameter() {defaultValue=0.02, group="[3] Segments", parameterName=SegmentEdgeSoftness, sliderMax=1, sortPriority=2}
  n61 = Constant()
  NamedRerouteDeclaration(None=If(A=NamedRerouteUsage(), "A > B"=@n58, "A == B"=If(A=@n59, "A > B"=@n60, "A == B"=@n61, "A < B"=@n60) {constB=1}, "A < B"=@n58) {constB=1}) {name=SegmentEdgeSoftness, nodeColor=<rgba 0 0 0 1>}
  NamedRerouteDeclaration(None=ScalarParameter() {group="[4] Color", parameterName=GradientOn, sliderMax=1, sortPriority=8}) {name=Scalar_GradientOn, nodeColor=<rgba 0 0 0 1>}
  n62 = NamedRerouteUsage()
  n63 = MaterialFunctionCall(GlowMin=NamedRerouteUsage(), GlowMax=Constant() {r=1}) {materialFunction=&MF_UI_SDF_Circle}
  n64 = NamedRerouteUsage()
  NamedRerouteDeclaration(None=MaterialFunctionCall(
    In=LinearInterpolate(A=@n62, B=@n63.Fill, Alpha=@n64)
    Value 1=NamedRerouteUsage()
    Value 2=NamedRerouteUsage()
    Value 3=NamedRerouteUsage()
    Pos value 1=ScalarParameter() {group="[4] Color", parameterName=Color1Position, sliderMax=1, sortPriority=3}
    Pos value 2=ScalarParameter() {defaultValue=0.5, group="[4] Color", parameterName=Color2Position, sliderMax=1, sortPriority=5}
    Pos value 3=ScalarParameter() {defaultValue=1, group="[4] Color", parameterName=Color3Position, sliderMax=1, sortPriority=7}
   ) {materialFunction=&MF_UI_RemappableGradient}.Gradient 3 Colors) {name=GradientMask, nodeColor=<rgba 0.7109 1 0 1>}
  NamedRerouteDeclaration(None=VectorParameter() {group="[4] Color", parameterName=Color1, sortPriority=2, defaultValue=<rgba 1 1 1 1>}.RGB) {name=FirstColor, nodeColor=<rgba 0 0 0 1>}
  NamedRerouteDeclaration(None=VectorParameter() {group="[4] Color", parameterName=Color2, sortPriority=4, defaultValue=<rgba 1 1 1 1>}.RGB) {name=SecondColor, nodeColor=<rgba 0 0 0 1>}
 ]
}
