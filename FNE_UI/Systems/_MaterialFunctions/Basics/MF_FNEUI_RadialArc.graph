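# Material function in FNE_UI that outputs radial arc masks for UI fills: Radial Fill, Centered Fill and Arc Mask.
# UV, defaulting to normalized UI UV, becomes an angle gradient through a cheaper vector-to-radial function.
# Start/end angle, fill ratio and clockwise inputs drive named reroutes; a remap with swappable target range forms the arc gradient.
# Saturated subtractions produce the fills, with a mirrored min(x, 1-x) gradient for centered filling.
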MaterialFunction /FNE_UI/Systems/_MaterialFunctions/Basics/MF_FNEUI_RadialArc
graph {nodes [
  n1 = NamedRerouteDeclaration(None=NamedRerouteUsage()) {name=Arc Mask, nodeColor=<rgba 0 0 0 1>}
  n2 = NamedRerouteUsage()
  FunctionOutput(None=Saturate(None=Subtract(A=@n1, B=@n2))) {description=Output the radial filled arc, outputName=Radial Fill}
  FunctionOutput(None=NamedRerouteUsage()) {description=Outputs the arc mask, outputName=Arc Mask, sortPriority=2}
  n3 = NamedRerouteUsage()
  n4 = Multiply(A=Min(A=@n3, B=OneMinus(None=@n3))) {constB=2}
  n5 = NamedRerouteUsage()
  FunctionOutput(None=Saturate(None=Subtract(A=@n4, B=@n5))) {description=Output a filled arc from the centered angle, outputName=Centered Fill, sortPriority=1}
  n6 = MaterialFunctionCall() {materialFunction=&GetUserInterfaceUV}
  NamedRerouteDeclaration(None=MaterialFunctionCall!(
    Vector or UVs=FunctionInput(Preview=@n6.Normalized UV) {bUsePreviewValueAsDefault=true, description=Sett the UV, inputName=UV, inputType=FunctionInput_Vector2}
    Swizzle Coordinate Output
   ) {materialFunction=&/MF_FNEUI_VectorToRadial_Cheaper}.Vector Converted to Angle) {name=Angle Gradient, nodeColor=<rgba 0 0 0 1>}
  n7 = FunctionInput(Preview=Constant()) {
   bUsePreviewValueAsDefault=true
   description="Set the filling value (0-1)"
   inputName="Fill Ratio (0-1)"
   inputType=FunctionInput_Scalar
   sortPriority=4
  }
  NamedRerouteDeclaration(None=@n7) {name=Fill Ratio, nodeColor=<rgba 0 0 0 1>}
  n8 = FunctionInput(Preview=Constant()) {
   bUsePreviewValueAsDefault=true
   description=Define the arc start angle
   inputName="Arc Start Angle (0-1)"
   inputType=FunctionInput_Scalar
   sortPriority=2
  }
  NamedRerouteDeclaration(None=@n8) {name=Arc Start Angle, nodeColor=<rgba 0 0 0 1>}
  n9 = FunctionInput(Preview=Constant()) {
   bUsePreviewValueAsDefault=true
   description=Define the arc end angle
   inputName="Arc End Angle (0-1)"
   inputType=FunctionInput_Scalar
   sortPriority=3
  }
  NamedRerouteDeclaration(None=@n9) {name=Arc End Angle, nodeColor=<rgba 0 0 0 1>}
  n10 = FunctionInput(Preview=Constant() {r=1}) {
   bUsePreviewValueAsDefault=true
   description="Define if the progress filling is clockwise or anticlockwise (only for radial fill)"
   inputName=Fill Clockwise
   inputType=FunctionInput_Scalar
   sortPriority=6
  }
  NamedRerouteDeclaration(None=@n10) {name=Is Clockwise, nodeColor=<rgba 0 0 0 1>}
  n11 = LinearInterpolate(Alpha=NamedRerouteUsage())
  MaterialExpressionRemap_0 = MaterialXRemap!(
   None=OneMinus(None=NamedRerouteUsage())
   InputLow=NamedRerouteUsage()
   InputHigh=OneMinus(None=NamedRerouteUsage())
   TargetLow=@n11
   TargetHigh=OneMinus(None=@n11)
  )
  n12 = OneMinus(None=Floor(None=@MaterialExpressionRemap_0))
  NamedRerouteDeclaration(None=Saturate(None=Multiply(A=@n12, B=@MaterialExpressionRemap_0))) {name=Remapped Angle Gradient, nodeColor=<rgba 0 0 0 1>}
 ]}
