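# Material function in FNE_UI that turns a distance field, from a procedural SDF or greyscale texture, into an anti-aliased edge mask around a Threshold.
# Edge width derives from the field's screen-space derivative length, combined with a remapped copy of it plus the Smoothing input, which softens the edge.
# The offset threshold difference is divided by that width and saturated, giving Inner for negative-inside SDFs and its inverse, Outer, for white-inside textures.
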MaterialFunction /FNE_UI/Systems/_MaterialFunctions/Shapes/MF_FNEUI_AntiAliasing
graph {nodes [
  n1 = FunctionInput() {
   bUsePreviewValueAsDefault=true
   description="The field value that anti-aliasing will occur around. Use \"0.0\" to anti-alias around the exact boundary of procedural SDFs. Use \"0.5\" to anti-alias around mid-gray for texture-based DFs."
   inputName=Threshold
   inputType=FunctionInput_Scalar
   sortPriority=1
  }
  n2 = FunctionInput(Preview=MaterialFunctionCall!(UVs, Radius) {materialFunction=/FNE_UI/Systems/_MaterialFunctions/SDFs/MF_FNEUI_SDF_Circle_Basic}.SDF) {
   description="The distance field (eg. SDF shape or greyscale DF texture) to perform anti-aliasing on."
   inputName=DistanceField
   inputType=FunctionInput_Scalar
  }
  n3 = Subtract(A=@n1, B=@n2)
  n4 = FunctionInput() {
   bUsePreviewValueAsDefault=true
   description="Blurs/softens the resulting edges. Note that adjusting this value will bypass anti-aliasing!"
   inputName=Smoothing
   inputType=FunctionInput_Scalar
   sortPriority=2
  }
  n5 = Length(None=AppendVector(A=DDX(Value=@n2), B=DDY(Value=@n2))) {is=class MaterialExpressionLength}
  n6 = Constant() {r=0.01}
  n7 = Constant() {r=1}
  n8 = Constant()
  n9 = Constant() {r=2}
  n10 = Add(
   A=@n4
   B=Saturate(None=MaterialFunctionCall(Input=@n5, Input Low=@n6, Input High=@n7, Target Low=@n8, Target High=@n9) {materialFunction=&RemapValueRange}.Result)
  ) {constB=0}
  n11 = Multiply(A=@n10) {constB=0.5}
  n12 = Max(A=@n5, B=@n10)
  n13 = Saturate(None=Divide(A=Add(A=@n3, B=@n11), B=Max(A=@n12) {constB=0, desc=Prevent Divide by Zero}))
  FunctionOutput(None=OneMinus(None=@n13)) {description=Works best with texture-based DFs where shape interior is white and exterior is black., outputName=Outer, sortPriority=1}
  FunctionOutput(None=@n13) {
   description=Works best with procedural SDF shapes where negative values indicate shape interior and positive values indicate the exterior.
   outputName=Inner
  }
 ]}
