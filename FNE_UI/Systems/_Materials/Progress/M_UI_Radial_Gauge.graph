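# M_UI_Radial_Gauge is a translucent UI material that draws a configurable arc gauge with progress fill, critical zone, tick segments and pointer.
# Progress, either normalized or remapped from a min/max range, drives the optionally gradient fill and rotates an arrow pointer paired with a circle.
# SDF shape functions and polar-coordinate gradients build masks, linked by named reroutes, that layer into Color and Opacity.
# Primary and secondary segments overlay the arc or cut out of it.
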Material /FNE_UI/Systems/_Materials/Progress/M_UI_Radial_Gauge
graph {
 properties {bAutomaticallySetUsageInEditor=false, blendMode=BLEND_Translucent, materialDomain=MD_UI, usedShadingModels=Default Lit}
 nodes [
  n1 = ScalarParameter() {
   defaultValue=0.5
   desc="Progress Value (should be between 0 and 1 when \"ProgressIsNormalized\" = 1 or between ProgressRangeMin and ProgressRandMax when “ProgressIsNormalized” = 0).\r\n"
   group="[1] Basics"
   parameterName=Progress
   sortPriority=0
  }
  NamedRerouteDeclaration(None=@n1) {name=Scalar_Progress, nodeColor=<rgba 0 0 0 1>}
  n2 = ScalarParameter() {
   defaultValue=1
   desc="Set to 1 when progress is between 0 and 1.\r\nSet to 0 when progress should be between ProgressRangeMin and ProgressRangeMax.\r\n"
   group="[1] Basics"
   parameterName=ProgressIsNormalized
   sliderMax=1
   sortPriority=1
  }
  NamedRerouteDeclaration(None=@n2) {name=Scalar_ProgressIsNormalized, nodeColor=<rgba 0 0 0 1>}
  n3 = NamedRerouteUsage()
  n4 = NamedRerouteUsage()
  NamedRerouteDeclaration(None=OneMinus(None=Step(Y=@n3, X=@n4))) {name=ProgressMask, nodeColor=<rgba 0 1 0.2734 1>}
  n5 = ScalarParameter() {
   defaultValue=0.1
   desc="Change the thickness of the background shape.\r\n"
   group="[2] Shape"
   parameterName=ShapeTickness
   sliderMax=1
   sortPriority=5
  }
  NamedRerouteDeclaration(None=@n5) {name=Scalar_ShapeThickness, nodeColor=<rgba 0 0 0 1>}
  n6 = NamedRerouteUsage()
  n7 = Multiply(A=NamedRerouteUsage(), B=@n6)
  n8 = Constant()
  n9 = If(A=@n6, "A > B"=NamedRerouteUsage(), "A == B"=@n8, "A < B"=@n8)
  NamedRerouteDeclaration(None=Multiply(
    A=MaterialFunctionCall!(UV, Size=@n7, GlowMin, GlowMax=@n9, Stroke thickness, Is stroke thickness relative to size) {materialFunction=/FNE_UI/Systems/_MaterialFunctions/SDFs/MF_UI_SDF_Circle}.Fill
    B=NamedRerouteUsage()
   )) {name=PointerCircleMask, nodeColor=<rgba 0 0 0 1>}
  n10 = NamedRerouteUsage()
  n11 = Subtract(A=NamedRerouteUsage(), B=@n10)
  n12 = Subtract(A=NamedRerouteUsage(), B=@n10)
  NamedRerouteDeclaration(None=LinearInterpolate(A=Divide(A=@n11, B=@n12), B=NamedRerouteUsage(), Alpha=NamedRerouteUsage())) {name=Scalar_CurrentProgress, nodeColor=<rgba 0 0 0 1>}
  n13 = ScalarParameter() {defaultValue=5, desc="Set the number of primary segments.\r\n", group="[5] Segments", parameterName=PrimarySegmentsCount, sortPriority=0}
  NamedRerouteDeclaration(None=Subtract(A=@n13) {constB=2}) {name=Scalar_PrimarySegmentsCount, nodeColor=<rgba 0 0 0 1>}
  n14 = ScalarParameter() {
   defaultValue=0.8
   desc="Set the distance between primary segments.\r\n"
   group="[5] Segments"
   parameterName=PrimarySegmentsSpacing
   sliderMax=1
   sortPriority=1
  }
  NamedRerouteDeclaration(None=@n14) {name=Scalar_PrimarySegmentsSpacing, nodeColor=<rgba 0 0 0 1>}
  n15 = ScalarParameter() {
   defaultValue=0.05
   desc="Change the sharpness of the edges of primary segments.\r\n"
   group="[5] Segments"
   parameterName=PrimarySegmentsEdgeSoftness
   sliderMax=0.5
   sortPriority=2
  }
  NamedRerouteDeclaration(None=@n15) {name=Scalar_PrimarySegmentsEdgeSoftness, nodeColor=<rgba 0 0 0 1>}
  n16 = NamedRerouteUsage()
  n17 = Multiply(A=Add(A=NamedRerouteUsage()), B=Add(A=NamedRerouteUsage()))
  n18 = NamedRerouteUsage()
  n19 = Min(
   A=MaterialFunctionCall!(
    DistanceField=MaterialFunctionCall!(UVs, Gradient Map=@n16, Sector Count=@n17, Sector Spacing=@n18, Fixed Spacing) {materialFunction=/FNE_UI/Systems/_MaterialFunctions/Basics/MF_FNEUI_Sectors}.Parallel Sectors SDF
    Threshold=Constant() {r=0.25}
    Smoothing=NamedRerouteUsage()
   ) {materialFunction=/FNE_UI/Systems/_MaterialFunctions/Shapes/MF_FNEUI_AntiAliasing}.Outer
   B=NamedRerouteUsage()
  )
  n20 = NamedRerouteUsage()
  n21 = Multiply(A=Saturate(None=Min(A=@n19, B=@n20)), B=NamedRerouteUsage())
  n22 = NamedRerouteUsage()
  n23 = MaterialFunctionCall(
   DistanceField=MaterialFunctionCall(Gradient Map=NamedRerouteUsage(), Sector Count=Add(A=@n22), Sector Spacing=NamedRerouteUsage()) {materialFunction=&MF_FNEUI_Sectors}.Parallel Sectors SDF
   Threshold=Constant() {r=0.25}
   Smoothing=NamedRerouteUsage()
  ) {materialFunction=&MF_FNEUI_AntiAliasing}
  n24 = MaterialFunctionCall(Size=NamedRerouteUsage(), GlowMax=NamedRerouteUsage(), Stroke thickness=NamedRerouteUsage()) {materialFunction=&MF_UI_SDF_Circle}
  n25 = Multiply(A=Min(A=Min(A=@n23.Outer, B=@n24.Outline), B=NamedRerouteUsage()), B=NamedRerouteUsage())
  NamedRerouteDeclaration(None=Max(
    A=NamedRerouteDeclaration(None=@n21) {name=SecondarySegments, nodeColor=<rgba 0 0 0 1>}
    B=NamedRerouteDeclaration(None=@n25) {name=PrimarySegments, nodeColor=<rgba 0 0 0 1>}
   )) {name=Segments, nodeColor=<rgba 0 0 0 1>}
  n26 = Min(
   A=Step(Y=Constant() {r=0}, X=@n4) {constY=0}
   B=MaterialFunctionCall(Size=NamedRerouteUsage(), GlowMax=NamedRerouteUsage(), Stroke thickness=NamedRerouteUsage()) {materialFunction=&MF_UI_SDF_Circle}.Outline
  )
  n27 = NamedRerouteDeclaration(None=@n26) {name=ProgressShape, nodeColor=<rgba 0.3359 1 0 1>}
  n28 = Min(A=NamedRerouteUsage(), B=@n27)
  n29 = NamedRerouteUsage()
  n30 = Multiply(A=@n27, B=@n29)
  n31 = NamedRerouteUsage()
  n32 = NamedRerouteUsage()
  n33 = NamedRerouteUsage()
  n34 = Max(A=Max(A=Max(A=@n28, B=@n30), B=Min(A=@n31, B=@n32)), B=LinearInterpolate(A=Min(A=@n26, B=@n33), B=@n27, Alpha=@n29))
  n35 = Subtract(A=@n34, B=@n33)
  n36 = NamedRerouteUsage()
  n37 = NamedRerouteUsage()
  n38 = NamedRerouteUsage()
  NamedRerouteDeclaration(None=Max(A=LinearInterpolate(A=@n34, B=@n35, Alpha=@n36) {constAlpha=1}, B=Max(A=@n37, B=@n38))) {name=Opacity, nodeColor=<rgba 0 0.9297 1 1>}
  n39 = VectorParameter() {
   desc="Set the critical zone color.\r\n"
   group="[4] Colors"
   parameterName=CriticalZoneColor
   sortPriority=6
   defaultValue=<rgba 1 0.4931 0 1>
  }
  n40 = VectorParameter() {
   desc="Set the circle background color.\r\n"
   group="[4] Colors"
   parameterName=ShapeColor
   sortPriority=0
   defaultValue=<rgba 0.0185 0.0082 0.1042 1>
  }
  n41 = LinearInterpolate(A=@n40.RGB, B=@n39.RGB, Alpha=NamedRerouteUsage())
  n42 = NamedRerouteUsage()
  n43 = VectorParameter() {desc="Set the progress color.\r\n", group="[4] Colors", parameterName=ProgressColor, sortPriority=3, defaultValue=<rgba 1 0 0.9794 1>}
  n44 = VectorParameter() {
   desc="Set the color of secondary segments.\r\n"
   group="[5] Segments"
   parameterName=SecondarySegmentsColor
   sortPriority=9
   defaultValue=<rgba 1 0.2899 0 1>
  }
  n45 = VectorParameter() {
   desc="Set the color of primary segments.\r\n"
   group="[5] Segments"
   parameterName=PrimarySegmentsColor
   sortPriority=3
   defaultValue=<rgba 0 0.8698 0.799 1>
  }
  n46 = LinearInterpolate(
   A=LinearInterpolate(A=LinearInterpolate(A=@n41, B=@n43.RGB, Alpha=@n42), B=@n44.RGB, Alpha=NamedRerouteUsage())
   B=@n45.RGB
   Alpha=NamedRerouteUsage()
  )
  n47 = VectorParameter() {
   desc="Set the color of the arrow pointer.\r\n"
   group="[3] Pointer"
   parameterName=PointerArrowColor
   sortPriority=3
   defaultValue=<rgba 1 1 1 1>
  }
  n48 = NamedRerouteUsage()
  n49 = VectorParameter() {
   desc="Set the color of the circle pointer.\r\n"
   group="[3] Pointer"
   parameterName=PointerCircleColor
   sortPriority=7
   defaultValue=<rgba 0.9809 1 0 1>
  }
  NamedRerouteDeclaration(None=LinearInterpolate(A=LinearInterpolate(A=@n46, B=@n47.RGB, Alpha=@n48), B=@n49.RGB, Alpha=NamedRerouteUsage())) {name=Color, nodeColor=<rgba 0 0 0 1>}
  n50 = ScalarParameter() {
   desc="Set to the minimum value of the progress range.\r\n\r\nOnly used when \"ProgressIsNormalized\" = 0.\r\n"
   group="[1] Basics"
   parameterName=ProgressRangeMin
   sortPriority=2
  }
  NamedRerouteDeclaration(None=@n50) {name=Scalar_ProgressRangeMin, nodeColor=<rgba 0 0 0 1>}
  n51 = ScalarParameter() {
   defaultValue=4
   desc="Set the number of secondary segments between primary segments.\r\n"
   group="[5] Segments"
   parameterName=SecondarySegmentsCount
   sortPriority=5
  }
  NamedRerouteDeclaration(None=@n51) {name=Scalar_SecondarySegmentsCount, nodeColor=<rgba 0 0 0 1>}
  n52 = ScalarParameter() {
   defaultValue=0.8
   desc="Set the distance between secondary segments.\r\n"
   group="[5] Segments"
   parameterName=SecondarySegmentsSpacing
   sliderMax=1
   sortPriority=6
  }
  NamedRerouteDeclaration(None=@n52) {name=Scalar_SecondarySegmentSpacing, nodeColor=<rgba 0 0 0 1>}
  n53 = ScalarParameter() {
   defaultValue=0.1
   desc="Change the sharpness of the secondary segments' edges.\r\n"
   group="[5] Segments"
   parameterName=SecondarySegmentsEdgeSoftness
   sliderMax=0.5
   sortPriority=8
  }
  NamedRerouteDeclaration(None=@n53) {name=Scalar_SecondarySegmentsEdgeSoftness, nodeColor=<rgba 0 0 0 1>}
  n54 = NamedRerouteUsage()
  n55 = MaterialFunctionCall(UVs=TextureCoordinate(), "Rotation Angle (0-1)"=Multiply(A=Add(A=@n54) {constB=0.5}) {constB=-1}) {materialFunction=&CustomRotator}
  n56 = NamedRerouteUsage()
  n57 = NamedRerouteUsage()
  n58 = NamedRerouteUsage()
  n59 = Subtract(A=OneMinus(None=NamedRerouteUsage()), B=Multiply(A=Subtract(A=@n57, B=@n58), B=NamedRerouteUsage()))
  n60 = Constant()
  n61 = LinearInterpolate(
   A=NamedRerouteUsage()
   B=LinearInterpolate(
    B=MaterialFunctionCall!(UV=@n55.Rotated Values, "Arc Start Angle (0-1)"=@n56, "Arc End Angle (0-1)"=@n59, "Fill Ratio (0-1)", Fill Clockwise=@n60) {materialFunction=/FNE_UI/Systems/_MaterialFunctions/Basics/MF_FNEUI_RadialArc}.Arc Mask
    Alpha=NamedRerouteUsage()
   )
   Alpha=NamedRerouteUsage()
  ) {constAlpha=0}
  n62 = NamedRerouteUsage()
  NamedRerouteDeclaration(None=Multiply(A=Min(A=@n61, B=@n62), B=NamedRerouteUsage())) {name=GradientFill, nodeColor=<rgba 0 0 0 1>}
  n63 = NamedRerouteUsage()
  n64 = NamedRerouteUsage()
  n65 = Subtract(A=@n63, B=@n64)
  n66 = Subtract(A=NamedRerouteUsage(), B=@n64)
  n67 = OneMinus(None=LinearInterpolate(A=Divide(A=@n65, B=@n66), B=@n63, Alpha=NamedRerouteUsage()))
  n68 = NamedRerouteUsage()
  NamedRerouteDeclaration(None=Min(A=Step(Y=@n67, X=@n68), B=NamedRerouteUsage())) {name=CriticalZone, nodeColor=<rgba 0 0 0 1>}
  n69 = ScalarParameter() {
   defaultValue=1
   desc="Set the transparency of secondary segments between 0 (transparent) and 1 (opaque).\r\n"
   group="[5] Segments"
   parameterName=SecondarySegmentsOpacity
   sliderMax=1
   sortPriority=10
  }
  NamedRerouteDeclaration(None=@n69) {name=Scalar_SecondarySegmentsOpacity, nodeColor=<rgba 0 0 0 1>}
  n70 = ScalarParameter() {
   defaultValue=0.5
   desc=Set the size of the secondary segments within the background shape thickness.
   group="[5] Segments"
   parameterName=SecondarySegmentsSize
   sliderMax=1
   sortPriority=7
  }
  NamedRerouteDeclaration(None=@n70) {name=Scalar_SecondarySegmentsSize, nodeColor=<rgba 0 0 0 1>}
  n71 = NamedRerouteUsage()
  n72 = NamedRerouteUsage()
  NamedRerouteDeclaration(None=MaterialFunctionCall(Size=NamedRerouteUsage(), GlowMax=NamedRerouteUsage(), Stroke thickness=Multiply(A=@n71, B=@n72)) {materialFunction=&MF_UI_SDF_Circle}.Outline) {name=MaskSecondarySegments, nodeColor=<rgba 0 0 0 1>}
  n73 = ScalarParameter() {
   defaultValue=1
   desc="Set to the maximum value of the progress range.\r\n\r\nOnly used when \"ProgressIsNormalized\" = 0.\r\n"
   group="[1] Basics"
   parameterName=ProgressRangeMax
   sortPriority=3
  }
  NamedRerouteDeclaration(None=@n73) {name=Scalar_ProgressRangeMax, nodeColor=<rgba 0 0 0 1>}
  n74 = ScalarParameter() {
   defaultValue=0.1
   desc="Set the size of the circle pointer attached to the arrow pointer.\r\n"
   group="[3] Pointer"
   parameterName=PointerCircleSize
   sliderMax=1
   sortPriority=5
  }
  NamedRerouteDeclaration(None=@n74) {name=Scalar_PointerCircleSize, nodeColor=<rgba 0 0 0 1>}
  n75 = ScalarParameter() {
   defaultValue=0.05
   desc="Change the sharpness of the pointer arrow's edges.\r\n"
   group="[3] Pointer"
   parameterName=PointerArrowEdgeSoftness
   sliderMax=0.2
   sortPriority=2
  }
  NamedRerouteDeclaration(None=@n75) {name=Scalar_PointerArrowEdgeSoftness, nodeColor=<rgba 0 0 0 1>}
  n76 = ScalarParameter() {
   defaultValue=0.2
   desc="Set the size of the critical zone within the circle.\r\n"
   group="[2] Shape"
   parameterName=CriticalZone
   sliderMax=0.8
   sortPriority=6
  }
  NamedRerouteDeclaration(None=@n76) {name=Scalar_CriticalZone, nodeColor=<rgba 0 0 0 1>}
  n77 = ScalarParameter() {
   defaultValue=1
   desc="0 - Gradient Off \r\nGreater than 0 = Gradient on\r\n\r\nWhen enabled, ProgressColor is used to make the gradient.\r\n"
   group="[4] Colors"
   parameterName=ProgressGradientOn
   sliderMax=1
   sortPriority=5
  }
  NamedRerouteDeclaration(None=@n77) {name=Scalar_ProgressGradientOn, nodeColor=<rgba 0 0 0 1>}
  n78 = ScalarParameter() {
   defaultValue=1
   desc="Set the transparency of primary segments between 0 (transparent) and 1 (opaque).\r\n"
   group="[5] Segments"
   parameterName=PrimarySegmentsOpacity
   sliderMax=1
   sortPriority=4
  }
  NamedRerouteDeclaration(None=@n78) {name=Scalar_PrimarySegmentsOpacity, nodeColor=<rgba 0 0 0 1>}
  n79 = ScalarParameter() {
   desc="0.0 - The segments are shown on the background shape.\r\n1.0 - The mask segments are subtracted from the background shape.\r\n"
   group="[5] Segments"
   parameterName=SegmentsMask
   sliderMax=1
   sortPriority=11
  }
  NamedRerouteDeclaration(None=@n79) {name=Scalar_SegmentsMask, nodeColor=<rgba 0 0 0 1>}
  n80 = ScalarParameter() {
   defaultValue=1
   desc="Set the transparency of the arrow pointer between 0 (transparent) and 1 (opaque).\r\n"
   group="[3] Pointer"
   parameterName=PointerArrowOpacity
   sliderMax=1
   sortPriority=4
  }
  NamedRerouteDeclaration(None=@n80) {name=Scalar_PointerArrowOpacity, nodeColor=<rgba 0 0 0 1>}
  n81 = ScalarParameter() {
   defaultValue=0.2
   desc="Create an arc from the circle by setting the start position of the arc.\r\n"
   group="[2] Shape"
   parameterName=PositionBegin
   sliderMax=1
   sortPriority=1
  }
  NamedRerouteDeclaration(None=@n81) {name=Scalar_PositionBegin, nodeColor=<rgba 0 0 0 1>}
  n82 = ScalarParameter() {
   defaultValue=1
   desc="Set the background circle's transparency between 0 (transparent) and 1 (opaque).\r\n"
   group="[4] Colors"
   parameterName=ShapeOpacity
   sliderMax=1
   sortPriority=1
  }
  NamedRerouteDeclaration(None=@n82) {name=Scalar_ShapeOpacity, nodeColor=<rgba 0 0 0 1>}
  n83 = ScalarParameter() {
   defaultValue=1
   desc="Set the progress transparency between 0 (transparent) and 1 (opaque).\r\n"
   group="[4] Colors"
   parameterName=ProgressOpacity
   sliderMax=1
   sortPriority=4
  }
  NamedRerouteDeclaration(None=@n83) {name=Scalar_ProgressOpacity, nodeColor=<rgba 0 0 0 1>}
  n84 = NamedRerouteUsage()
  n85 = AppendVector(A=Constant() {r=1}, B=Constant() {r=0.5})
  n86 = AppendVector(A=Constant() {r=0.5}, B=Constant())
  n87 = Constant() {r=0.5}
  n88 = NamedRerouteUsage()
  n89 = NamedRerouteUsage()
  n90 = NamedRerouteUsage()
  n91 = Add(A=Add(A=LinearInterpolate(A=@n88, B=@n89, Alpha=@n90) {constA=1, constAlpha=0, constB=0}, B=NamedRerouteUsage())) {constB=0.5}
  n92 = NamedRerouteDeclaration(None=MaterialFunctionCall(UVs=TextureCoordinate(), Rotation Center=AppendVector(A=@n87, B=@n87), "Rotation Angle (0-1)"=Multiply(A=@n91) {constB=-1}) {materialFunction=&CustomRotator}.Rotated Values) {name=ProgressRotation, nodeColor=<rgba 1 0 0.75 1>}
  n93 = AppendVector(A=NamedRerouteUsage(), B=NamedRerouteUsage())
  n94 = Constant() {r=0.5}
  n95 = AppendVector(A=@n94, B=@n94)
  n96 = Min(
   A=MaterialFunctionCall!(
    UVs=MaterialFunctionCall!(UVs=@n84, Scale=@n85, Scale center=@n86) {materialFunction=/FNE_UI/Systems/_MaterialFunctions/Transforms/MF_UI_Scale}.Result
    Size=Constant() {r=1}
    Outline Thickness
    Glow Min
    Glow Max=NamedRerouteUsage()
    Corner Roundness=Constant()
    Keep Aspect Ratio
   ) {materialFunction=/FNE_UI/Systems/_MaterialFunctions/Shapes/MF_FNEUI_Box}.Fill
   B=MaterialFunctionCall!(
    UVs=MaterialFunctionCall(UVs=@n92, Scale=@n93, Scale center=@n95) {materialFunction=&MF_UI_Scale}.Result
    Radius=NamedRerouteUsage()
    Glow Min
    Glow Max=NamedRerouteUsage()
    Corner roundness
    Stroke thickness
    Is stroke thickness relative to size
   ) {materialFunction=/FNE_UI/Systems/_MaterialFunctions/SDFs/MF_UI_SDF_Triangle}.Fill
  )
  n97 = Subtract(A=NamedRerouteUsage(), B=NamedRerouteUsage())
  n98 = Subtract(A=NamedRerouteUsage(), B=@n97)
  n99 = Multiply(A=NamedRerouteUsage()) {constB=0.9}
  n100 = Multiply(A=NamedRerouteUsage(), B=Add(A=Multiply(A=@n98, B=@n99) {constB=0.95}, B=@n97))
  n101 = NamedRerouteUsage()
  n102 = Constant()
  n103 = OneMinus(None=MaterialFunctionCall(Size=@n100, GlowMax=If(A=@n100, "A > B"=@n101, "A == B"=@n102, "A < B"=@n102)) {materialFunction=&MF_UI_SDF_Circle}.Fill)
  NamedRerouteDeclaration(None=Multiply(A=Min(A=@n96, B=@n103), B=NamedRerouteUsage())) {name=PointerMask, nodeColor=<rgba 0 0 0 1>}
  n104 = ScalarParameter() {
   defaultValue=0.8
   desc="Create an arc from the circle by setting the end position of the arc.\r\n"
   group="[2] Shape"
   parameterName=PositionEnd
   sliderMax=1
   sortPriority=2
  }
  NamedRerouteDeclaration(None=@n104) {name=Scalar_PositionEnd, nodeColor=<rgba 0 0 0 1>}
  n105 = ScalarParameter() {
   defaultValue=1
   desc="Set the transparency of the circle pointer between 0 (transparent) and 1 (opaque).\r\n"
   group="[3] Pointer"
   parameterName=PointerCircleOpacity
   sliderMax=1
   sortPriority=8
  }
  NamedRerouteDeclaration(None=@n105) {name=Scalar_PointerCircleOpacity, nodeColor=<rgba 0 0 0 1>}
  n106 = ScalarParameter() {
   defaultValue=0.5
   desc="Set the height of the arrow pointer.\r\n"
   group="[3] Pointer"
   parameterName=PointerArrowSizeY
   sliderMax=1.5
   sortPriority=0
  }
  NamedRerouteDeclaration(None=@n106) {name=Scalar_PointerArrowSizeY, nodeColor=<rgba 0 0 0 1>}
  n107 = ScalarParameter() {
   desc="Set the distance between the arrow and circle pointers.\r\n"
   group="[3] Pointer"
   parameterName=PointerGap
   sliderMax=1
   sortPriority=9
  }
  NamedRerouteDeclaration(None=@n107) {name=Scalar_PointerGap, nodeColor=<rgba 0 0 0 1>}
  n108 = ScalarParameter() {
   defaultValue=0.18
   desc="Set the width of the arrow pointer.\r\n"
   group="[3] Pointer"
   parameterName=PointerArrowSizeX
   sliderMax=1
   sortPriority=1
  }
  NamedRerouteDeclaration(None=@n108) {name=Scalar_PointerArrowSizeX, nodeColor=<rgba 0 0 0 1>}
  n109 = ScalarParameter() {
   defaultValue=0.01
   desc="Change the sharpness of the pointer circle’s edges.\r\n"
   group="[3] Pointer"
   parameterName=PointerCircleEdgeSoftness
   sliderMax=0.2
   sortPriority=6
  }
  NamedRerouteDeclaration(None=@n109) {name=Scalar_PointerCircleEdgeSoftness, nodeColor=<rgba 0 0 0 1>}
  n110 = ScalarParameter() {desc="Rotate the entire shape.\r\n", group="[2] Shape", parameterName=Rotation, sliderMax=1, sliderMin=-1, sortPriority=3}
  NamedRerouteDeclaration(None=@n110) {name=Scalar_Rotation, nodeColor=<rgba 0 0 0 1>}
  n111 = ScalarParameter() {
   defaultValue=0.95
   desc="Scale the entire shape from 0 to 1 (100%).\r\n"
   group="[2] Shape"
   parameterName=Scale
   sliderMax=1
   sortPriority=4
  }
  NamedRerouteDeclaration(None=@n111) {name=Scalar_Scale, nodeColor=<rgba 0 0 0 1>}
  n112 = ScalarParameter() {defaultValue=0.01, desc="Change how sharp the edges are.\r\n", group="[2] Shape", parameterName=EdgeSoftness, sliderMax=0.2, sortPriority=0}
  NamedRerouteDeclaration(None=@n112) {name=Scalar_EdgeSoftness, nodeColor=<rgba 0 0 0 1>}
  n113 = Add(A=NamedRerouteUsage()) {constB=0.5}
  n114 = MaterialFunctionCall!(
   Vector or UVs=MaterialFunctionCall(UVs=TextureCoordinate(), Rotation Center=Constant() {r=0.5}, "Rotation Angle (0-1)"=Multiply(A=@n113) {constB=-1}) {materialFunction=&CustomRotator}.Rotated Values
   Swizzle Coordinate Output
  ) {materialFunction=/FNE_UI/Systems/_MaterialFunctions/Distortions/MF_UI_PolarCoordinates}
  n115 = Constant()
  n116 = NamedRerouteUsage()
  n117 = NamedRerouteUsage()
  n118 = Divide(A=Subtract(A=@n117, B=NamedRerouteUsage()))
  n119 = NamedRerouteUsage()
  NamedRerouteDeclaration(None=MaterialFunctionCall!(
    In=OneMinus(None=@n114.Angle gradient)
    Value 1=@n115
    Value 2=@n115
    Value 3=Constant() {r=0.5}
    Value 4=Constant() {r=1}
    Value 5=Constant()
    Pos value 1=@n116
    Pos value 2=@n116
    Pos value 3=Subtract(A=@n117, B=@n118)
    Pos value 4=@n119
    Pos value 5=@n119
   ) {materialFunction=/FNE_UI/Systems/_MaterialFunctions/Utilities/MF_UI_RemappableGradient}.Gradient 5 Colors) {name=ArcMask, nodeColor=<rgba 0 0 0 1>}
  NamedRerouteUsage()
  NamedRerouteUsage()
 ]
}
