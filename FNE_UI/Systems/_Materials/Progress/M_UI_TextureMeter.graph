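# Translucent UI material for a progress meter shaped by the red channel of a texture, by default a distance field.
# Raw or range-normalized progress drives a soft-edged fill mask with direction, rotation and slant controls.
# Derivative-based smoothstep thresholds antialias the shape and its stroke at BarSize and StrokeSize.
# Color layers combine bar and fill colors, an offset shine, a pulsing warning below WarningPercent, and panning Overlay-blended textures for fill and background.
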Material /FNE_UI/Systems/_Materials/Progress/M_UI_TextureMeter
graph {
 properties {bAutomaticallySetUsageInEditor=false, blendMode=BLEND_Translucent, materialDomain=MD_UI, usedShadingModels=Default Lit}
 nodes [
  n1 = VectorParameter() {group="[4] Colors", parameterName=BarColor, defaultValue=<rgba 0.0176 0.0176 0.0176 1>}
  n2 = VectorParameter() {group="[4] Colors", parameterName=FillColor1, defaultValue=<rgba 1 1 1 1>}
  n3 = VectorParameter() {group="[4] Colors", parameterName=FillColor2OrStrokeColor, defaultValue=<rgba 0.5 0.5 0.5 1>}
  n4 = LinearInterpolate(A=@n2.RGB, B=@n3.RGB, Alpha=ComponentMask(None=NamedRerouteUsage()) {r=true})
  n5 = Subtract(A=NamedRerouteUsage(), B=NamedRerouteUsage())
  n6 = Add(
   A=MaterialFunctionCall(UVs=NamedRerouteUsage(), Rotation Center=Constant2Vector() {g=0.5, r=0.5}, "Rotation Angle (0-1)"=Constant() {r=-0.02}) {materialFunction=&CustomRotator}.Rotated Values
   B=ScalarParameter() {defaultValue=-1, group="[1] Basics", parameterName=ShineOffset, sliderMax=1, sliderMin=-1}
  )
  n7 = LinearInterpolate(
   A=LinearInterpolate(
    A=@n1.RGB
    B=VectorParameter() {group="[4] Colors", parameterName=WarningColor, defaultValue=<rgba 0.1667 0 0.0045 1>}.RGB
    Alpha=NamedRerouteUsage()
   )
   B=LinearInterpolate(
    A=LinearInterpolate(A=@n4, B=@n3.RGB, Alpha=@n5) {desc=InnerOutline}
    B=VectorParameter() {group="[4] Colors", parameterName=ShineColor, defaultValue=<rgba 0.396 0.9597 1 1>}.RGB
    Alpha=TextureSample(UVs=@n6) {samplerSource=SSM_Clamp_WorldGroupSettings, texture=/FNE_UI/Systems/_Textures/T_UI_Shine}.R
   )
   Alpha=NamedRerouteUsage()
  )
  n8 = LinearInterpolate(
   A=@n7
   B=MaterialFunctionCall(Base=@n7, Blend=NamedRerouteUsage()) {materialFunction=&Blend_Overlay}.Result
   Alpha=Min(A=NamedRerouteUsage(), B=NamedRerouteUsage())
  )
  n9 = NamedRerouteUsage()
  LinearInterpolate(
   A=@n8
   B=MaterialFunctionCall(Base=@n8, Blend=NamedRerouteUsage()) {materialFunction=&Blend_Overlay}.Result
   Alpha=Min(A=OneMinus(None=@n9), B=NamedRerouteUsage())
  )
  n10 = NamedRerouteUsage()
  n11 = NamedRerouteUsage()
  Multiply(A=LinearInterpolate(A=@n10, B=Add(A=@n10, B=@n11), Alpha=NamedRerouteUsage()), B=NamedRerouteUsage())
  n12 = MaterialFunctionCall() {materialFunction=&GetUserInterfaceUV}
  n13 = NamedRerouteUsage()
  n14 = ScalarParameter() {group="[6] Slant", parameterName=FillSlantX, sliderMax=1, sliderMin=-1}
  n15 = ScalarParameter() {group="[6] Slant", parameterName=FillSlantY, sliderMax=1, sliderMin=-1}
  NamedRerouteDeclaration(None=MaterialFunctionCall!(
    UVs=MaterialFunctionCall(UVs=@n12.Normalized UV, "Rotation Angle (0-1)"=@n13) {materialFunction=&CustomRotator}.Rotated Values
    Slant amount=MaterialFunctionCall(X=@n14, Y=@n15) {materialFunction=&MakeFloat2}.Result
   ) {materialFunction=/FNE_UI/Systems/_MaterialFunctions/Effects/MF_FNEUI_Slant}.Slanted UVs) {name=NormalizedUV, nodeColor=<rgba 0 1 0.1094 1>}
  NamedRerouteDeclaration(None=MaterialFunctionCall(UVs=@n12.9-Slice UV, "Rotation Angle (0-1)"=NamedRerouteUsage()) {materialFunction=&CustomRotator}.Rotated Values) {name=9-SliceUV, nodeColor=<rgba 1 0 0.2109 1>}
  n16 = ScalarParameter() {group="[2] Fill Texture Overlay", parameterName=FillTextureOverlayPanningSpeedX, sliderMax=1, sliderMin=-1}
  n17 = ScalarParameter() {group="[2] Fill Texture Overlay", parameterName=FillTextureOverlayPanningSpeedY, sliderMax=1, sliderMin=-1}
  n18 = MaterialFunctionCall(
   UVs=Panner(
    Coordinate=TextureCoordinate()
    Speed=MaterialFunctionCall(X=@n16, Y=@n17) {materialFunction=&MakeFloat2}.Result
   )
   "Rotation Angle (0-1)"=Add(A=NamedRerouteUsage(), B=ScalarParameter() {group="[2] Fill Texture Overlay", parameterName=FillTextureOverlayRotation, sliderMax=1})
  ) {materialFunction=&CustomRotator}
  n19 = AppendVector(
   A=ScalarParameter() {group="[2] Fill Texture Overlay", parameterName=FillTextureOverlayOffsetX}
   B=ScalarParameter() {group="[2] Fill Texture Overlay", parameterName=FillTextureOverlayOffsetY}
  )
  n20 = ScalarParameter() {defaultValue=1, group="[2] Fill Texture Overlay", parameterName=FillTextureOverlayScaleX}
  n21 = ScalarParameter() {defaultValue=1, group="[2] Fill Texture Overlay", parameterName=FillTextureOverlayScaleY}
  n22 = TextureSampleParameter2D(UVs=MaterialFunctionCall!(
    UVs=MaterialFunctionCall!(Input UVs=@n18.Rotated Values, Translate=@n19) {materialFunction=/FNE_UI/Systems/_MaterialFunctions/Transforms/MF_UI_Translate}.UVs raw
    Scale=AppendVector(A=@n20, B=@n21)
    Scale center
   ) {materialFunction=/FNE_UI/Systems/_MaterialFunctions/Transforms/MF_UI_Scale}.Result) {
   desc=Applies a texture on top of the fill of the progress bar.
   group="[2] Fill Texture Overlay"
   parameterName=FillTextureOverlay
   texture=/FNE_UI/Systems/_Textures/T_DissolveFire
  }
  NamedRerouteDeclaration(None=Multiply(A=@n22.RGB, B=VectorParameter() {group="[4] Colors", parameterName=FillTextureOverlayColor, defaultValue=<rgba 1 1 1 1>}.RGB)) {name=FillTextureOverlay, nodeColor=<rgba 0.9219 0 1 1>}
  n23 = Fmod(
   A=Multiply(
    A=MaterialFunctionCall!(FloatValue, Speed) {materialFunction=/FNE_UI/Systems/_MaterialFunctions/Time/MF_FNEUI_LinearTime}.0-1 Loop
    B=ScalarParameter() {defaultValue=2, group="[5] WarningState", parameterName=PulseSpeed, sliderMax=20}
   )
   B=Constant() {r=1}
  )
  n24 = LinearInterpolate(A=NamedRerouteUsage(), B=NamedRerouteUsage(), Alpha=NamedRerouteUsage())
  n25 = ScalarParameter() {group="[5] WarningState", parameterName=WarningPercent, sliderMax=1}
  n26 = Constant()
  n27 = Constant() {r=1}
  NamedRerouteDeclaration(None=Multiply(
    A=MaterialFunctionCall(Value=@n23) {materialFunction=&LinearSine}."Rounded Linear Sine "
    B=If(A=@n24, B=@n25, "A > B"=@n26, "A == B"=@n26, "A < B"=@n27)
   )) {name=WarningState, nodeColor=<rgba 1 0.1172 0 1>}
  NamedRerouteDeclaration(None=ScalarParameter() {group="[1] Basics", parameterName=Rotation, sliderMax=1}) {name=Scalar_Rotation, nodeColor=<rgba 0 0 0 1>}
  NamedRerouteDeclaration(None=Multiply(A=@n22.A, B=ScalarParameter() {group="[2] Fill Texture Overlay", parameterName=FillTextureOverlayOpacity, sliderMax=1})) {name=FillTextureOverlayAlpha, nodeColor=<rgba 0.9453 0 1 1>}
  n28 = ScalarParameter() {group="[3] Background Texture Overlay", parameterName=BGTextureOverlayPanningSpeedX, sliderMax=1, sliderMin=-1}
  n29 = ScalarParameter() {group="[3] Background Texture Overlay", parameterName=BGTextureOverlayPanningSpeedY, sliderMax=1, sliderMin=-1}
  n30 = MaterialFunctionCall(
   UVs=Panner(
    Coordinate=TextureCoordinate()
    Speed=MaterialFunctionCall(X=@n28, Y=@n29) {materialFunction=&MakeFloat2}.Result
   )
   "Rotation Angle (0-1)"=Add(A=NamedRerouteUsage(), B=ScalarParameter() {group="[3] Background Texture Overlay", parameterName=BGTextureOverlayRotation, sliderMax=1})
  ) {materialFunction=&CustomRotator}
  n31 = AppendVector(
   A=ScalarParameter() {group="[3] Background Texture Overlay", parameterName=BGTextureOverlayOffsetX}
   B=ScalarParameter() {group="[3] Background Texture Overlay", parameterName=BGTextureOverlayOffsetY}
  )
  n32 = ScalarParameter() {defaultValue=1, group="[3] Background Texture Overlay", parameterName=BGTextureOverlayScaleX}
  n33 = ScalarParameter() {defaultValue=1, group="[3] Background Texture Overlay", parameterName=BGTextureOverlayScaleY}
  n34 = TextureSampleParameter2D(UVs=MaterialFunctionCall(
    UVs=MaterialFunctionCall(Input UVs=@n30.Rotated Values, Translate=@n31) {materialFunction=&MF_UI_Translate}.UVs raw
    Scale=AppendVector(A=@n32, B=@n33)
   ) {materialFunction=&MF_UI_Scale}.Result) {
   desc=Applies a texture to the background of the progress bar.
   group="[3] Background Texture Overlay"
   parameterName=BackgroundTextureOverlay
  }
  NamedRerouteDeclaration(None=Multiply(A=@n34.RGB, B=VectorParameter() {group="[4] Colors", parameterName=BGTextureOverlayColor, defaultValue=<rgba 1 1 1 1>}.RGB)) {name=BGTextureOverlay, nodeColor=<rgba 0.9219 0 1 1>}
  NamedRerouteDeclaration(None=Multiply(A=@n34.A, B=ScalarParameter() {group="[3] Background Texture Overlay", parameterName=BGTextureOverlayOpacity, sliderMax=1})) {name=BGTextureOverlayAlpha, nodeColor=<rgba 0.9453 0 1 1>}
  n35 = ScalarParameter() {
   desc="0.0 = Fill progress bar from bottom/left\r\n1.0 = Fill progress bar from top/right"
   group="[1] Basics"
   parameterName=FillDirection
   sliderMax=1
  }
  n36 = MaterialFunctionCall(Float2=NamedRerouteUsage()) {materialFunction=&BreakOutFloat2Components}
  n37 = ScalarParameter() {
   desc="0.0 = Fill progress bar horizontally\r\n1.0 = Fill progress bar vertically"
   group="[1] Basics"
   parameterName=FillVertical
   sliderMax=1
  }
  n38 = LinearInterpolate(A=@n36.R, B=@n36.R, Alpha=@n37)
  n39 = OneMinus(None=@n38)
  NamedRerouteDeclaration(None=If(A=Constant(), B=@n35, "A > B"=@n39, "A == B"=@n39, "A < B"=@n38)) {name=DirectionAndOrientationMask, nodeColor=<rgba 0 0.2031 1 1>}
  n40 = Constant3Vector() {constant=<rgba 0 1 0 0>}
  n41 = NamedRerouteUsage()
  n42 = NamedRerouteUsage()
  n43 = NamedRerouteUsage()
  n44 = OneMinus(None=LinearInterpolate(A=@n40.R, B=@n40.R, Alpha=LinearInterpolate(A=@n41, B=@n42, Alpha=@n43)) {constA=0.0381, constB=0.9619, desc=Adjust Clamp for bar shape})
  n45 = ScalarParameter() {defaultValue=0.005, group="[1] Basics", parameterName=EdgeSoftness, sliderMax=0.1}
  NamedRerouteDeclaration(None=SmoothStep(Min=Subtract(A=@n44, B=@n45) {constB=0.005}, Max=@n44, Value=NamedRerouteUsage())) {name=ProgressAlpha, nodeColor=<rgba 0.0312 0 1 1>}
  n46 = ScalarParameter() {
   desc="Progress Value (should be between 0 and 1 when \"ProgressIsNormalized\" = 1)"
   group="[1] Basics"
   parameterName=Progress
   sortPriority=0
  }
  NamedRerouteDeclaration(None=@n46) {name=Scalar_Progress, nodeColor=<rgba 0 0 0 1>}
  n47 = ScalarParameter() {
   defaultValue=1
   desc="Set to the maximum value of the progress range.\r\n\r\nOnly used when \"ProgressIsNormalized\" = 0"
   group="[1] Basics"
   parameterName=ProgressRangeMax
   sortPriority=3
  }
  NamedRerouteDeclaration(None=@n47) {name=Scalar_ProgressMax, nodeColor=<rgba 0 0 0 1>}
  NamedRerouteDeclaration(None=ScalarParameter() {defaultValue=0.9, group="[1] Basics", parameterName=BarShapeSoftness, sliderMax=1}) {name=Scalar_BarShapeSoftness, nodeColor=<rgba 0 0 0 1>}
  n48 = ScalarParameter() {
   desc="Set to the minimum value of the progress range. \r\n\r\nOnly used when \"ProgressIsNormalized\" = 0"
   group="[1] Basics"
   parameterName=ProgressRangeMin
   sortPriority=2
  }
  NamedRerouteDeclaration(None=@n48) {name=Scalar_ProgressMin, nodeColor=<rgba 0 0 0 1>}
  n49 = NamedRerouteUsage()
  n50 = NamedRerouteUsage()
  n51 = NamedRerouteUsage()
  NamedRerouteDeclaration(None=Divide(A=Subtract(A=@n49, B=@n50), B=Subtract(A=@n51, B=@n50))) {name=NormalizeProgress, nodeColor=<rgba 0 0.6484 1 1>}
  NamedRerouteDeclaration(None=ScalarParameter() {defaultValue=0.22, group="[1] Basics", parameterName=BarSize, sliderMax=1, sliderMin=0.01}) {name=Scalar_BarSize, nodeColor=<rgba 0 0 0 1>}
  n52 = Multiply(
   A=NamedRerouteUsage()
   B=AppendVector(
    A=ScalarParameter() {defaultValue=1, group="[1] Basics", parameterName=TextureSizeWidth, sliderMax=1}
    B=ScalarParameter() {defaultValue=1, group="[1] Basics", parameterName=TextureSizeHeight, sliderMax=1}
   )
  )
  n53 = ScalarParameter() {defaultValue=1, group="[1] Basics", parameterName=BarScale, sliderMax=1}
  n54 = ScalarParameter() {group="[1] Basics", parameterName=TextureOffsetX}
  n55 = ScalarParameter() {group="[1] Basics", parameterName=TextureOffsetY}
  n56 = TextureSampleParameter2D(UVs=MaterialFunctionCall(
    Input UVs=MaterialFunctionCall(UVs=@n52, Texture Scale=@n53) {materialFunction=&ScaleUVsByCenter}.UVs
    Translate=AppendVector(A=@n54, B=@n55)
   ) {materialFunction=&MF_UI_Translate}.UVs raw) {group="[1] Basics", parameterName=Texture, samplerSource=SSM_Clamp_WorldGroupSettings, texture=/FNE_UI/Systems/_Textures/T_UI_SDF_Dist}
  NamedRerouteDeclaration(None=@n56.R) {name=TextureChannelR, nodeColor=<rgba 0 1 0.9062 1>}
  n57 = Constant() {r=3}
  n58 = NamedRerouteUsage()
  n59 = NamedRerouteUsage()
  n60 = DDX(Value=NamedRerouteUsage())
  n61 = DDY(Value=NamedRerouteUsage())
  n62 = Multiply(A=Add(A=Abs(None=@n60), B=Abs(None=@n61)), B=NamedRerouteUsage())
  n63 = Subtract(A=@n59, B=@n62)
  n64 = Add(A=@n59, B=@n62)
  n65 = Clamp(None=Divide(A=Subtract(A=@n58, B=@n63), B=Subtract(A=@n64, B=@n63)))
  n66 = Multiply(A=Constant() {r=2}, B=@n65)
  NamedRerouteDeclaration(None=Multiply(A=Subtract(A=@n57, B=@n66), B=Multiply(A=@n65, B=@n65))) {name=AAThreshold, nodeColor=<rgba 1 0.5859 0 1>}
  NamedRerouteDeclaration(None=@n1.A) {name=BarColorAlpha, nodeColor=<rgba 0.2656 0 1 1>}
  NamedRerouteDeclaration(None=@n2.A) {name=FillColorAplha, nodeColor=<rgba 0.0547 1 0 1>}
  n67 = Constant() {r=3}
  n68 = NamedRerouteUsage()
  n69 = Add(A=NamedRerouteUsage(), B=ScalarParameter() {defaultValue=0.048, group="[1] Basics", parameterName=StrokeSize, sliderMax=1})
  n70 = NamedRerouteUsage()
  n71 = DDX(Value=@n70)
  n72 = DDY(Value=@n70)
  n73 = Multiply(A=Add(A=Abs(None=@n71), B=Abs(None=@n72)), B=NamedRerouteUsage())
  n74 = Subtract(A=@n69, B=@n73)
  n75 = Add(A=@n69, B=@n73)
  n76 = Clamp(None=Divide(A=Subtract(A=@n68, B=@n74), B=Subtract(A=@n75, B=@n74)))
  n77 = Multiply(A=Constant() {r=2}, B=@n76)
  NamedRerouteDeclaration(None=Multiply(A=Subtract(A=@n67, B=@n77), B=Multiply(A=@n76, B=@n76))) {name=AAThreshold_Stroke, nodeColor=<rgba 1 0 0.9375 1>}
  n78 = ScalarParameter() {
   defaultValue=1
   desc=Set to 1 when progress is between 0 and 1.
   group="[1] Basics"
   parameterName=ProgressIsNormalized
   sliderMax=1
   sortPriority=1
  }
  NamedRerouteDeclaration(None=@n78) {name=Scalar_Normalized, nodeColor=<rgba 0 0 0 1>}
 ]
}
